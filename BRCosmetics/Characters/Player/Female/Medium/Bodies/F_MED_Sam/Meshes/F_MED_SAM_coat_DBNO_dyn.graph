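# Physics asset for the F_MED_Sam coat in its DBNO state, simulating coat and accessory dynamics.
# Five coat chains (left, right, back, back-left, back-right) hang from spine_04 via master bodies with limited swing and tight twist.
# Angularly free cross-links tie neighboring chains mid-length so the panels move together.
# Pelvis, spine, thigh and upper-arm shapes form the body; bowknot and single-axis pouch attach to pelvis, bag to left thigh.
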PhysicsAsset /BRCosmetics/Characters/Player/Female/Medium/Bodies/F_MED_Sam/Meshes/F_MED_SAM_coat_DBNO_dyn
bodies {
 dyn_bag_MSTR [{shapeType=Box, extentX=18, extentY=10.3005, extentZ=4.2712, center=<6.6146 0.0002 2.2705>, rotation=<rot 1.4137 0 5>}]
 dyn_bowknot_MSTR [{shapeType=Capsule, length=4.9798, radius=2.5775, center=<0 -0.1533 0.5719>, rotation=<rot 0 0 84.9974>}]
 dyn_coat_bk_01 [{shapeType=Capsule, length=8, radius=6, shapeName=dyn_coat_bk_01_capsule, center=<5 0 3.5>, rotation=<rot 0 0 90>}]
 dyn_coat_bk_02 [{shapeType=Capsule, length=8, radius=6, shapeName=dyn_coat_bk_02_capsule, center=<5 0 3.5>, rotation=<rot 0 0 90>}]
 dyn_coat_bk_03 [{shapeType=Capsule, length=8, radius=6, shapeName=dyn_coat_bk_03_capsule, center=<5 0 3.5>, rotation=<rot 0 0 90>}]
 dyn_coat_bk_04 [{shapeType=Capsule, length=8, radius=6, shapeName=dyn_coat_bk_04_capsule, center=<5 0 3.5>, rotation=<rot 0 0 90>}]
 dyn_coat_bk_05 [{shapeType=Capsule, length=8, radius=6, shapeName=dyn_coat_bk_05_capsule, center=<5 0 3.5>, rotation=<rot 0 0 90>}]
 dyn_coat_bk_MSTR [{shapeType=Capsule, length=0.5, radius=0.5, rotation=<rot 0 -90 0>}]
 dyn_coat_l_01 [{shapeType=Box, extentX=10, extentY=5, extentZ=1.5, center=<5 0 0>, rotation=<rot 0 0 -10>}]
 dyn_coat_l_02 [{shapeType=Capsule, length=3.4, radius=3.5, shapeName=dyn_coat_l_02_capsule, center=<5 0 2>, rotation=<rot 90 0 0>}]
 dyn_coat_l_03 [{shapeType=Capsule, length=9, radius=6, shapeName=dyn_coat_l_03_capsule, center=<5.5 -1.8 3.5>, rotation=<rot 0 0 -104.9995>}]
 dyn_coat_l_04 [{shapeType=Capsule, length=8, radius=6, shapeName=dyn_coat_l_04_capsule, center=<5.5 -1.8 3.5>, rotation=<rot 0 0 -105.0004>}]
 dyn_coat_l_05 [{shapeType=Capsule, length=8, radius=6, shapeName=dyn_coat_l_05_capsule, center=<5.5 -1.8 3.5>, rotation=<rot 0 0 -110>}]
 dyn_coat_l_06 [{shapeType=Capsule, length=17.0427, radius=7.4168, shapeName=dyn_coat_l_06_capsule, center=<5.5 -1.8 3.5>, rotation=<rot 0 0 -105>}]
 dyn_coat_l_MSTR [{shapeType=Capsule, length=0.5, radius=0.5, rotation=<rot 0 -90 0>}]
 dyn_coat_l_bk_01 [{shapeType=Capsule, length=8, radius=6, shapeName=dyn_coat_l_bk_01_capsule, center=<5 0 3.5>, rotation=<rot 0 0 90>}]
 dyn_coat_l_bk_02 [{shapeType=Capsule, length=8, radius=6, shapeName=dyn_coat_l_bk_02_capsule, center=<5 0 3.5>, rotation=<rot 0 0 90>}]
 dyn_coat_l_bk_03 [{shapeType=Capsule, length=8, radius=6, shapeName=dyn_coat_l_bk_03_capsule, center=<5 0 3.5>, rotation=<rot 0 0 90>}]
 dyn_coat_l_bk_04 [{shapeType=Capsule, length=8, radius=6, shapeName=dyn_coat_l_bk_04_capsule, center=<5 0 3.5>, rotation=<rot 0 0 90>}]
 dyn_coat_l_bk_05 [{shapeType=Capsule, length=8, radius=6, shapeName=dyn_coat_l_bk_05_capsule, center=<5 0 3.5>, rotation=<rot 0 0 90>}]
 dyn_coat_l_bk_MSTR [{shapeType=Capsule, length=0.5, radius=0.5, rotation=<rot 0 -90 0>}]
 dyn_coat_r_01 [{shapeType=Box, extentX=10, extentY=5, extentZ=1.5, center=<5 0 0>, rotation=<rot 0 0 10>}]
 dyn_coat_r_02 [{shapeType=Capsule, length=3.3856, radius=3.5666, shapeName=dyn_coat_r_02_capsule, center=<5 0 2>, rotation=<rot 90 0 0>}]
 dyn_coat_r_03 [{shapeType=Capsule, length=8.9314, radius=5.0686, shapeName=dyn_coat_r_03_capsule, center=<5.5 1.8 3.5>, rotation=<rot 0 0 114.9998>}]
 dyn_coat_r_04 [{shapeType=Capsule, length=8, radius=6, shapeName=dyn_coat_r_04_capsule, center=<5.5 1.8 3.5>, rotation=<rot 0 0 110>}]
 dyn_coat_r_05 [{shapeType=Capsule, length=8, radius=6, shapeName=dyn_coat_r_05_capsule, center=<5.5 1.8 3.5>, rotation=<rot 0 0 110>}]
 dyn_coat_r_06 [{shapeType=Capsule, length=18, radius=7.5, shapeName=dyn_coat_r_06_capsule, center=<5.5 1.8 3.5>, rotation=<rot 0 0 110.0023>}]
 dyn_coat_r_MSTR [{shapeType=Capsule, length=0.5, radius=0.5, rotation=<rot 0 -90 0>}]
 dyn_coat_r_bk_01 [{shapeType=Capsule, length=8, radius=6, shapeName=dyn_coat_r_bk_01_capsule, center=<5 0 3.5>, rotation=<rot 0 0 90>}]
 dyn_coat_r_bk_02 [{shapeType=Capsule, length=8, radius=6, shapeName=dyn_coat_r_bk_02_capsule, center=<5 0 3.5>, rotation=<rot 0 0 90>}]
 dyn_coat_r_bk_03 [{shapeType=Capsule, length=8, radius=6, shapeName=dyn_coat_r_bk_03_capsule, center=<5 0 3.5>, rotation=<rot 0 0 90>}]
 dyn_coat_r_bk_04 [{shapeType=Capsule, length=8, radius=6, shapeName=dyn_coat_r_bk_04_capsule, center=<5 0 3.5>, rotation=<rot 0 0 90>}]
 dyn_coat_r_bk_05 [{shapeType=Capsule, length=8, radius=6, shapeName=dyn_coat_r_bk_05_capsule, center=<5 0 3.5>, rotation=<rot 0 0 90>}]
 dyn_coat_r_bk_MSTR [{shapeType=Capsule, length=0.5, radius=0.5, rotation=<rot 0 -90 0>}]
 dyn_pouch_base [
  {shapeType=Capsule, length=8.7178, radius=3.0984, shapeName=dyn_pouch_base_capsule, center=<1.5251 0.2936 1.6064>, rotation=<rot 0 0 89.9994>}
  {shapeType=Capsule, length=6.7854, radius=9.4274, shapeName=dyn_pouch_base_capsule, center=<10.5585 0.0002 -1.711>, rotation=<rot 90 0 0>}
 ]
 dyn_pouch_base_MSTR [{shapeType=Capsule, length=0.5, radius=0.5}]
 pelvis [
  {shapeType=Capsule, length=12.153, radius=10.097, center=<-2.2905 -2.8179 -0.0002>}
  {shapeType=Capsule, length=13.653, radius=10.097, center=<-6.6286 -2.8179 -0.0005>}
 ]
 spine_01 [{shapeType=Capsule, length=8.25, radius=9.25, center=<0 -0.8942 -0.0001>}]
 spine_02 [{shapeType=Capsule, length=6, radius=8.5, center=<0 0.0192 -0.0001>}]
 spine_03 [{shapeType=Capsule, length=4, radius=8.75, center=<-0.0001 1.1951 -0.0001>}]
 spine_04 [{shapeType=Capsule, length=1.5, radius=9.75, center=<1.1747 4.2278 0.0001>}]
 spine_05 [{shapeType=Capsule, length=7, radius=7.5, center=<0 0.904 0>}]
 thigh_l [{shapeType=Capsule, length=24.1395, radius=8.6105, center=<-12.4112 -0.6651 0.9577>, rotation=<rot 87.7595 180 180>}]
 thigh_r [
  {shapeType=Capsule, length=21.75, radius=7.25, center=<12.369 -0.0811 -1.0452>, rotation=<rot 87.1996 0 -180>}
  {shapeType=Capsule, length=14.75, radius=8, center=<15.1378 0.8611 -0.1068>, rotation=<rot 87.1797 0 -180>}
 ]
 upperarm_twist_01_l [{shapeType=Capsule, length=11.25, radius=4.25, center=<8 0 0>, rotation=<rot 90 0 0>}]
 upperarm_twist_01_r [{shapeType=Capsule, length=11.25, radius=4.25, center=<-8 0 0>, rotation=<rot 90 0 0>}]
}
constraints [
 {bone1Name=spine_01, bone2Name=pelvis, swing1LimitDegrees=45, swing2LimitDegrees=45, twistLimitDegrees=45}
 {bone1Name=spine_03, bone2Name=spine_02, swing1LimitDegrees=45, swing2LimitDegrees=45, twistLimitDegrees=45}
 {bone1Name=spine_05, bone2Name=spine_04, swing1LimitDegrees=45, swing2LimitDegrees=45, twistLimitDegrees=45}
 {
  bone1Name=spine_04
  bone2Name=spine_03
  swing1LimitDegrees=45
  swing1Motion=Free
  swing2LimitDegrees=45
  swing2Motion=Free
  twistLimitDegrees=45
  twistMotion=Free
 }
 {
  bone1Name=spine_02
  bone2Name=spine_01
  swing1LimitDegrees=45
  swing1Motion=Free
  swing2LimitDegrees=45
  swing2Motion=Free
  twistLimitDegrees=45
  twistMotion=Free
 }
 {
  bone1Name=dyn_coat_l_MSTR
  bone2Name=spine_04
  swing1LimitDegrees=45
  swing1Motion=Free
  swing2LimitDegrees=45
  swing2Motion=Free
  twistLimitDegrees=45
  twistMotion=Free
 }
 {bone1Name=dyn_coat_l_01, bone2Name=dyn_coat_l_MSTR, swing1LimitDegrees=20, swing2LimitDegrees=15, twistLimitDegrees=5}
 {bone1Name=dyn_coat_l_02, bone2Name=dyn_coat_l_01, swing1LimitDegrees=45, swing2LimitDegrees=15, twistLimitDegrees=5}
 {bone1Name=dyn_coat_l_03, bone2Name=dyn_coat_l_02, swing1LimitDegrees=45, swing2LimitDegrees=60, twistLimitDegrees=3}
 {bone1Name=dyn_coat_l_04, bone2Name=dyn_coat_l_03, swing1LimitDegrees=45, swing2LimitDegrees=60, twistLimitDegrees=3}
 {bone1Name=dyn_coat_l_05, bone2Name=dyn_coat_l_04, swing1LimitDegrees=45, swing2LimitDegrees=60, twistLimitDegrees=3}
 {bone1Name=dyn_coat_l_06, bone2Name=dyn_coat_l_05, swing1LimitDegrees=45, swing2LimitDegrees=60, twistLimitDegrees=3}
 {
  bone1Name=dyn_coat_bk_MSTR
  bone2Name=spine_04
  swing1LimitDegrees=45
  swing1Motion=Free
  swing2LimitDegrees=45
  swing2Motion=Free
  twistLimitDegrees=45
  twistMotion=Free
 }
 {bone1Name=dyn_coat_bk_01, bone2Name=dyn_coat_bk_MSTR, swing1LimitDegrees=45, swing2LimitDegrees=30, twistLimitDegrees=5}
 {bone1Name=dyn_coat_bk_02, bone2Name=dyn_coat_bk_01, swing1LimitDegrees=45, swing2LimitDegrees=60, twistLimitDegrees=3}
 {bone1Name=dyn_coat_bk_03, bone2Name=dyn_coat_bk_02, swing1LimitDegrees=45, swing2LimitDegrees=60, twistLimitDegrees=3}
 {bone1Name=dyn_coat_bk_04, bone2Name=dyn_coat_bk_03, swing1LimitDegrees=45, swing2LimitDegrees=60, twistLimitDegrees=3}
 {bone1Name=dyn_coat_bk_05, bone2Name=dyn_coat_bk_04, swing1LimitDegrees=45, swing2LimitDegrees=60, twistLimitDegrees=3}
 {
  bone1Name=dyn_coat_l_bk_MSTR
  bone2Name=spine_04
  swing1LimitDegrees=45
  swing1Motion=Free
  swing2LimitDegrees=45
  swing2Motion=Free
  twistLimitDegrees=45
  twistMotion=Free
 }
 {bone1Name=dyn_coat_l_bk_01, bone2Name=dyn_coat_l_bk_MSTR, swing1LimitDegrees=45, swing2LimitDegrees=30, twistLimitDegrees=5}
 {bone1Name=dyn_coat_l_bk_02, bone2Name=dyn_coat_l_bk_01, swing1LimitDegrees=45, swing2LimitDegrees=60, twistLimitDegrees=3}
 {bone1Name=dyn_coat_l_bk_03, bone2Name=dyn_coat_l_bk_02, swing1LimitDegrees=45, swing2LimitDegrees=60, twistLimitDegrees=3}
 {bone1Name=dyn_coat_l_bk_04, bone2Name=dyn_coat_l_bk_03, swing1LimitDegrees=45, swing2LimitDegrees=60, twistLimitDegrees=3}
 {bone1Name=dyn_coat_l_bk_05, bone2Name=dyn_coat_l_bk_04, swing1LimitDegrees=45, swing2LimitDegrees=60, twistLimitDegrees=3}
 {
  bone1Name=dyn_coat_r_MSTR
  bone2Name=spine_04
  swing1LimitDegrees=45
  swing1Motion=Free
  swing2LimitDegrees=45
  swing2Motion=Free
  twistLimitDegrees=45
  twistMotion=Free
 }
 {bone1Name=dyn_coat_r_01, bone2Name=dyn_coat_r_MSTR, swing1LimitDegrees=20, swing2LimitDegrees=15, twistLimitDegrees=5}
 {bone1Name=dyn_coat_r_02, bone2Name=dyn_coat_r_01, swing1LimitDegrees=20, swing2LimitDegrees=15, twistLimitDegrees=5}
 {bone1Name=dyn_coat_r_03, bone2Name=dyn_coat_r_02, swing1LimitDegrees=45, swing2LimitDegrees=60, twistLimitDegrees=3}
 {bone1Name=dyn_coat_r_04, bone2Name=dyn_coat_r_03, swing1LimitDegrees=45, swing2LimitDegrees=60, twistLimitDegrees=3}
 {bone1Name=dyn_coat_r_05, bone2Name=dyn_coat_r_04, swing1LimitDegrees=45, swing2LimitDegrees=60, twistLimitDegrees=3}
 {bone1Name=dyn_coat_r_06, bone2Name=dyn_coat_r_05, swing1LimitDegrees=45, swing2LimitDegrees=60, twistLimitDegrees=3}
 {
  bone1Name=dyn_coat_r_bk_MSTR
  bone2Name=spine_04
  swing1LimitDegrees=45
  swing1Motion=Free
  swing2LimitDegrees=45
  swing2Motion=Free
  twistLimitDegrees=45
  twistMotion=Free
 }
 {bone1Name=dyn_coat_r_bk_01, bone2Name=dyn_coat_r_bk_MSTR, swing1LimitDegrees=45, swing2LimitDegrees=30, twistLimitDegrees=5}
 {bone1Name=dyn_coat_r_bk_02, bone2Name=dyn_coat_r_bk_01, swing1LimitDegrees=45, swing2LimitDegrees=60, twistLimitDegrees=3}
 {bone1Name=dyn_coat_r_bk_03, bone2Name=dyn_coat_r_bk_02, swing1LimitDegrees=45, swing2LimitDegrees=60, twistLimitDegrees=3}
 {bone1Name=dyn_coat_r_bk_04, bone2Name=dyn_coat_r_bk_03, swing1LimitDegrees=45, swing2LimitDegrees=60, twistLimitDegrees=3}
 {bone1Name=dyn_coat_r_bk_05, bone2Name=dyn_coat_r_bk_04, swing1LimitDegrees=45, swing2LimitDegrees=60, twistLimitDegrees=3}
 {
  bone1Name=thigh_r
  bone2Name=pelvis
  swing1LimitDegrees=45
  swing1Motion=Free
  swing2LimitDegrees=45
  swing2Motion=Free
  twistLimitDegrees=45
  twistMotion=Free
 }
 {
  bone1Name=thigh_l
  bone2Name=pelvis
  swing1LimitDegrees=45
  swing1Motion=Free
  swing2LimitDegrees=45
  swing2Motion=Free
  twistLimitDegrees=45
  twistMotion=Free
 }
 {
  bone1Name=dyn_bowknot_MSTR
  bone2Name=pelvis
  swing1LimitDegrees=45
  swing1Motion=Free
  swing2LimitDegrees=45
  swing2Motion=Free
  twistLimitDegrees=45
  twistMotion=Free
 }
 {
  bone1Name=dyn_pouch_base_MSTR
  bone2Name=pelvis
  swing1LimitDegrees=45
  swing1Motion=Free
  swing2LimitDegrees=45
  swing2Motion=Free
  twistLimitDegrees=45
  twistMotion=Free
 }
 {
  bone1Name=dyn_pouch_base
  bone2Name=dyn_pouch_base_MSTR
  swing1LimitDegrees=45
  swing1Motion=Locked
  swing2LimitDegrees=20
  twistLimitDegrees=45
  twistMotion=Locked
 }
 {
  bone1Name=dyn_bag_MSTR
  bone2Name=thigh_l
  swing1LimitDegrees=45
  swing1Motion=Free
  swing2LimitDegrees=45
  swing2Motion=Free
  twistLimitDegrees=45
  twistMotion=Free
 }
 {
  bone1Name=upperarm_twist_01_r
  bone2Name=spine_05
  swing1LimitDegrees=45
  swing1Motion=Free
  swing2LimitDegrees=45
  swing2Motion=Free
  twistLimitDegrees=45
  twistMotion=Free
 }
 {
  bone1Name=upperarm_twist_01_l
  bone2Name=spine_05
  swing1LimitDegrees=45
  swing1Motion=Free
  swing2LimitDegrees=45
  swing2Motion=Free
  twistLimitDegrees=45
  twistMotion=Free
 }
 {
  bone1Name=dyn_coat_l_04
  bone2Name=dyn_coat_l_bk_03
  swing1LimitDegrees=30
  swing1Motion=Free
  swing2LimitDegrees=45
  swing2Motion=Free
  twistLimitDegrees=5
  twistMotion=Free
 }
 {
  bone1Name=dyn_coat_l_bk_03
  bone2Name=dyn_coat_bk_03
  swing1LimitDegrees=30
  swing1Motion=Free
  swing2LimitDegrees=45
  swing2Motion=Free
  twistLimitDegrees=5
  twistMotion=Free
 }
 {
  bone1Name=dyn_coat_r_bk_03
  bone2Name=dyn_coat_bk_03
  swing1LimitDegrees=30
  swing1Motion=Free
  swing2LimitDegrees=45
  swing2Motion=Free
  twistLimitDegrees=5
  twistMotion=Free
 }
 {
  bone1Name=dyn_coat_r_04
  bone2Name=dyn_coat_r_bk_03
  swing1LimitDegrees=30
  swing1Motion=Free
  swing2LimitDegrees=45
  swing2Motion=Free
  twistLimitDegrees=5
  twistMotion=Free
 }
]
