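# Physics asset for the pouches dynamic mesh on the F_MED_Sam player body cosmetic, covering dangling coat and ribbon bones.
# Capsule chains form the coat's back, right and right-back strips off spine_01, while two box-shaped ribbon chains hang from a bowknot on the pelvis.
# Pelvis, spine_01 and right thigh capsules supply body shapes.
# Ribbons and the right-back strip have angular limits with tight twist; the other coat strips swing and twist freely.
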PhysicsAsset /BRCosmetics/Characters/Player/Female/Medium/Bodies/F_MED_Sam/Meshes/F_MED_SAM_pouches_dyn
bodies {
 dyn_bowknot_MSTR [{shapeType=Capsule, length=4.9798, radius=2.5775, center=<0 -0.1533 0.5719>, rotation=<rot 0 0 84.9974>}]
 dyn_coat_bk_03 [{shapeType=Capsule, length=13.6227, radius=7.2212, shapeName=dyn_coat_bk_03_capsule, center=<5 0 6>, rotation=<rot 0 0 90>}]
 dyn_coat_bk_04 [{shapeType=Capsule, length=13.6227, radius=7.2212, shapeName=dyn_coat_bk_04_capsule, center=<5 0 6>, rotation=<rot 0 0 90>}]
 dyn_coat_bk_05 [{shapeType=Capsule, length=13.6227, radius=7.2212, shapeName=dyn_coat_bk_05_capsule, center=<5 0 6>, rotation=<rot 0 0 90>}]
 dyn_coat_r_03 [{shapeType=Capsule, length=19.969, radius=7.5434, shapeName=dyn_coat_r_03_capsule, center=<5 1.8 5>, rotation=<rot 0 0 109.9991>}]
 dyn_coat_r_04 [{shapeType=Capsule, length=19.969, radius=7.5434, shapeName=dyn_coat_r_04_capsule, center=<5 1.8 5>, rotation=<rot 0 0 109.9987>}]
 dyn_coat_r_05 [{shapeType=Capsule, length=21.219, radius=6.2934, shapeName=dyn_coat_r_05_capsule, center=<5 1.8 5>, rotation=<rot 0 -0.0001 100.0005>}]
 dyn_coat_r_06 [{shapeType=Capsule, length=21.219, radius=6.2934, shapeName=dyn_coat_r_06_capsule, center=<5 1.8 5>, rotation=<rot 0 -0.0001 100.0005>}]
 dyn_coat_r_MSTR [{shapeType=Capsule, length=0, radius=1.6705, rotation=<rot 0 -90 0>}]
 dyn_coat_r_bk_03 [{shapeType=Capsule, length=10.6359, radius=7.3798, shapeName=dyn_coat_r_bk_03_capsule, center=<5 0 6>, rotation=<rot 0 0 90>}]
 dyn_coat_r_bk_04 [{shapeType=Capsule, length=11.8859, radius=6.1298, shapeName=dyn_coat_r_bk_04_capsule, center=<5 0 6>, rotation=<rot 0 0 90>}]
 dyn_coat_r_bk_05 [{shapeType=Capsule, length=11.8859, radius=6.1298, shapeName=dyn_coat_r_bk_05_capsule, center=<5 0 6>, rotation=<rot 0 0 90>}]
 dyn_ribbon_a_01 [{shapeType=Box, extentX=4.7567, extentY=4, extentZ=0.5, center=<3 0 0>}]
 dyn_ribbon_a_02 [{shapeType=Box, extentX=10.3733, extentY=4, extentZ=0.5, center=<5.5 0 0>, rotation=<rot 0 0 -10.0003>}]
 dyn_ribbon_a_03 [{shapeType=Box, extentX=10.0348, extentY=4, extentZ=0.7753, center=<6 0 0>, rotation=<rot 0 0 -10.0002>}]
 dyn_ribbon_b_01 [{shapeType=Box, extentX=4.5704, extentY=3.5, extentZ=0.993, center=<3.5344 0 0>}]
 dyn_ribbon_b_02 [{shapeType=Box, extentX=10.2252, extentY=4.5, extentZ=1, center=<5.5 0 0>, rotation=<rot 0 0 -10.0006>}]
 dyn_ribbon_b_03 [{shapeType=Box, extentX=11.1945, extentY=4.5, extentZ=1.344, center=<6 0 0>, rotation=<rot 0 0 -15.0006>}]
 pelvis [
  {shapeType=Capsule, length=12.153, radius=10.097, center=<-2.2905 -2.8179 -0.0002>}
  {shapeType=Capsule, length=13.153, radius=10.597, center=<-6.6286 -2.8179 -0.0005>}
 ]
 spine_01 [{shapeType=Capsule, length=8.25, radius=9.25, center=<0 -0.8942 -0.0001>}]
 thigh_r [
  {shapeType=Capsule, length=21.75, radius=7.25, center=<12.369 -0.0811 -1.0452>, rotation=<rot 87.1996 0 -180>}
  {shapeType=Capsule, length=14.75, radius=8, center=<15.1378 0.8611 -0.1068>, rotation=<rot 87.1797 0 -180>}
 ]
}
constraints [
 {bone1Name=spine_01, bone2Name=pelvis, swing1LimitDegrees=45, swing2LimitDegrees=45, twistLimitDegrees=45}
 {
  bone1Name=dyn_coat_r_MSTR
  bone2Name=spine_01
  swing1LimitDegrees=45
  swing1Motion=Free
  swing2LimitDegrees=45
  swing2Motion=Free
  twistLimitDegrees=45
  twistMotion=Free
 }
 {bone1Name=dyn_coat_r_bk_03, bone2Name=spine_01, swing1LimitDegrees=45, swing2LimitDegrees=60, twistLimitDegrees=3}
 {bone1Name=dyn_coat_r_bk_04, bone2Name=dyn_coat_r_bk_03, swing1LimitDegrees=45, swing2LimitDegrees=60, twistLimitDegrees=3}
 {bone1Name=dyn_coat_r_bk_05, bone2Name=dyn_coat_r_bk_04, swing1LimitDegrees=45, swing2LimitDegrees=60, twistLimitDegrees=3}
 {
  bone1Name=thigh_r
  bone2Name=pelvis
  swing1LimitDegrees=45
  swing1Motion=Free
  swing2LimitDegrees=45
  swing2Motion=Free
  twistLimitDegrees=45
  twistMotion=Free
 }
 {
  bone1Name=dyn_bowknot_MSTR
  bone2Name=pelvis
  swing1LimitDegrees=45
  swing1Motion=Free
  swing2LimitDegrees=45
  swing2Motion=Free
  twistLimitDegrees=45
  twistMotion=Free
 }
 {bone1Name=dyn_ribbon_a_01, bone2Name=dyn_bowknot_MSTR, swing1LimitDegrees=20, swing2LimitDegrees=60, twistLimitDegrees=5}
 {bone1Name=dyn_ribbon_a_02, bone2Name=dyn_ribbon_a_01, swing1LimitDegrees=20, swing2LimitDegrees=60, twistLimitDegrees=5}
 {bone1Name=dyn_ribbon_a_03, bone2Name=dyn_ribbon_a_02, swing1LimitDegrees=20, swing2LimitDegrees=60, twistLimitDegrees=5}
 {bone1Name=dyn_ribbon_b_01, bone2Name=dyn_bowknot_MSTR, swing1LimitDegrees=20, swing2LimitDegrees=60, twistLimitDegrees=5}
 {bone1Name=dyn_ribbon_b_02, bone2Name=dyn_ribbon_b_01, swing1LimitDegrees=20, swing2LimitDegrees=60, twistLimitDegrees=5}
 {bone1Name=dyn_ribbon_b_03, bone2Name=dyn_ribbon_b_02, swing1LimitDegrees=20, swing2LimitDegrees=60, twistLimitDegrees=5}
 {
  bone1Name=dyn_coat_r_06
  bone2Name=dyn_coat_r_05
  swing1LimitDegrees=45
  swing1Motion=Free
  swing2LimitDegrees=45
  swing2Motion=Free
  twistLimitDegrees=45
  twistMotion=Free
 }
 {
  bone1Name=dyn_coat_r_04
  bone2Name=dyn_coat_r_03
  swing1LimitDegrees=45
  swing1Motion=Free
  swing2LimitDegrees=45
  swing2Motion=Free
  twistLimitDegrees=45
  twistMotion=Free
 }
 {
  bone1Name=dyn_coat_r_05
  bone2Name=dyn_coat_r_04
  swing1LimitDegrees=45
  swing1Motion=Free
  swing2LimitDegrees=45
  swing2Motion=Free
  twistLimitDegrees=45
  twistMotion=Free
 }
 {
  bone1Name=dyn_coat_r_03
  bone2Name=dyn_coat_r_MSTR
  swing1LimitDegrees=45
  swing1Motion=Free
  swing2LimitDegrees=45
  swing2Motion=Free
  twistLimitDegrees=45
  twistMotion=Free
 }
 {
  bone1Name=dyn_coat_bk_05
  bone2Name=dyn_coat_bk_04
  swing1LimitDegrees=45
  swing1Motion=Free
  swing2LimitDegrees=45
  swing2Motion=Free
  twistLimitDegrees=45
  twistMotion=Free
 }
 {
  bone1Name=dyn_coat_bk_03
  bone2Name=spine_01
  swing1LimitDegrees=45
  swing1Motion=Free
  swing2LimitDegrees=45
  swing2Motion=Free
  twistLimitDegrees=45
  twistMotion=Free
 }
 {
  bone1Name=dyn_coat_bk_04
  bone2Name=dyn_coat_bk_03
  swing1LimitDegrees=45
  swing1Motion=Free
  swing2LimitDegrees=45
  swing2Motion=Free
  twistLimitDegrees=45
  twistMotion=Free
 }
]
