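# Post-process style AnimBlueprint for the F_MED_Sam cosmetic body, adding secondary motion to coat, pouches, ribbons, cuffs and bracelets.
# It copies the parent mesh pose, then applies weighted bone constraints and spring-based AnimDynamics to accessories.
# Coat branches start from cached preCoatPose, selecting RigidBody setups for carried DBNO, backpack and skydiving states.
# When rigid bodies are disabled, bone offsets and movement-driven Trail controllers move the coat instead.
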AnimBlueprint /BRCosmetics/Characters/Player/Female/Medium/Bodies/F_MED_Sam/Meshes/F_MED_SAM_AnimBP
target_skeleton=&/F_MED_SAM_Skeleton
graphs [
 AnimGraph {nodes [
   n1 = "Constraint - dyn_coat_l_02 (P)"(Alpha=1, ConstraintWeights_0=1) {
    alphaBoolBlend {blendOption=Linear}
    boneToModify {boneName=dyn_coat_l_02}
    constraintSetup [{targetBone {boneName=spine_02}}]
    constraintWeights [1]
   }
   "Constraint - dyn_coat_l_03 (P)"(ComponentPose=@n1, Alpha=1, ConstraintWeights_0=1) {
    alphaBoolBlend {blendOption=Linear}
    boneToModify {boneName=dyn_coat_l_03}
    constraintSetup [{targetBone {boneName=pelvis}}]
    constraintWeights [1]
   }
   n2 = "Constraint - dyn_coat_l_03 (P)"(Alpha=0.7, ConstraintWeights_0=1) {
    alphaBoolBlend {blendOption=Linear}
    boneToModify {boneName=dyn_coat_l_03}
    constraintSetup [{targetBone {boneName=pelvis}}]
    constraintWeights [1]
   }
   "Constraint - dyn_coat_l_bk_03 (P)"(ComponentPose=@n2, Alpha=0.5, ConstraintWeights_0=1) {
    alphaBoolBlend {blendOption=Linear}
    boneToModify {boneName=dyn_coat_l_bk_03}
    constraintSetup [{targetBone {boneName=pelvis}}]
    constraintWeights [1]
   }
   Look At - Bone: dyn_bracelet_l_02(Alpha=1) {
    alphaBoolBlend {blendOption=Linear}
    boneToModify {boneName=dyn_bracelet_l_02}
    lookAtLocation {x=1}
    lookAtTarget {boneReference {boneName=middle_metacarpal_l}}
    lookAt_Axis {axis=<1 0>}
    lookUp_Axis {axis {x=0}}
   }
   Look At - Bone: dyn_bracelet_l_01(Alpha=1) {
    alphaBoolBlend {blendOption=Linear}
    boneToModify {boneName=dyn_bracelet_l_01}
    lookAtLocation {x=0}
    lookAtTarget {boneReference {boneName=middle_metacarpal_l}}
    lookAt_Axis {axis=<1 0>}
   }
   "Transform (Modify) Bone - Bone: dyn_root_simspace"(Alpha=1, Translation="0,0,0", Rotation=Get sim_space_rotation() -> (sim_space_rotation:Rotator), Scale="1,1,1") {
    rotationMode=BMM_Replace
    rotationSpace=BCS_WorldSpace
    translationSpace=BCS_BoneSpace
    alphaBoolBlend {blendOption=Linear}
    alphaScaleBiasClamp {bInterpResult=true, interpSpeedDecreasing=2, interpSpeedIncreasing=2}
    boneToModify {boneName=dyn_root_simspace}
   }
   n3 = "Constraint - dyn_coat_l_bk_03 (P)"(Alpha=1, ConstraintWeights_0=1) {
    alphaBoolBlend {blendOption=Linear}
    boneToModify {boneName=dyn_coat_l_bk_03}
    constraintSetup [{targetBone {boneName=pelvis}}]
    constraintWeights [1]
   }
   "Transform (Modify) Bone - Bone: dyn_coat_l_03"(ComponentPose=@n3, Alpha=1, Translation="0,0,0", Rotation="0.000,-7.463,0.000", Scale="1,1,1") {
    rotationMode=BMM_Additive
    rotationSpace=BCS_BoneSpace
    translationMode=BMM_Additive
    translationSpace=BCS_BoneSpace
    alphaBoolBlend {blendOption=Linear}
    boneToModify {boneName=dyn_coat_l_03}
    rotation {yaw=-7.4631}
   }
   n4 = Get gravity_override() -> (gravity_override:Vector)
   n5 = RigidBody!(ComponentPose=Local To Component(LocalPose=Use cached pose 'preCoatPose'()), Alpha, OverrideWorldGravity:Vector=@n4, ExternalForce) -> (Pose) {
    bClampLinearTranslationLimitToRefPose=true
    bOverrideWorldGravity=true
    overridePhysicsAsset=&/F_MED_SAM_coat_skydive_dyn
    simulationSpace=BaseBoneSpace
    alphaBoolBlend {blendOption=Linear}
    baseBoneRef {boneName=pelvis}
    componentAppliedLinearAccClamp=<3000 3000 3000>
    componentLinearAccScale=<0.3 0.3 0.3>
    componentLinearVelScale=<0.3 0.3 0.3>
    overrideWorldGravity {z=-980}
    simSpaceSettings {
     maxAngularAcceleration=200
     maxAngularVelocity=5
     maxLinearAcceleration=1000
     maxLinearVelocity=300
     velocityScaleZ=0
     worldAlpha=1
     externalLinearDragV=<1 1>
    }
   }
   n6 = RigidBody(ComponentPose=@n5, OverrideWorldGravity=@n4) {
    bClampLinearTranslationLimitToRefPose=true
    bOverrideWorldGravity=true
    overridePhysicsAsset=&/F_MED_SAM_pouches_dyn
    simulationSpace=BaseBoneSpace
    alphaBoolBlend {blendOption=Linear}
    baseBoneRef {boneName=dyn_root_simspace}
    componentAppliedLinearAccClamp=<1500 1500 1500>
    componentLinearAccScale=<0.3 0.3 0.35>
    componentLinearVelScale=<0.3 0.3 0.3>
    overrideWorldGravity {z=-980}
    simSpaceSettings {
     maxAngularAcceleration=200
     maxAngularVelocity=5
     maxLinearAcceleration=1000
     maxLinearVelocity=300
     velocityScaleZ=0
     worldAlpha=1
     externalLinearDragV=<1 1>
    }
   }
   n7 = RigidBody!(ComponentPose=Local To Component(LocalPose=Use cached pose 'preCoatPose'()), bAlphaBoolEnabled:bool, OverrideWorldGravity=@n4, ExternalForce) -> (Pose) {
    alphaInputType=Bool
    bClampLinearTranslationLimitToRefPose=true
    bOverrideWorldGravity=true
    overridePhysicsAsset=&/F_MED_SAM_coat_2_dyn
    simulationSpace=BaseBoneSpace
    alphaBoolBlend {blendInTime=0.1, blendOption=Linear, blendOutTime=0.1}
    baseBoneRef {boneName=pelvis}
    componentAppliedLinearAccClamp=<3000 3000 3000>
    componentLinearAccScale=<0.3 0.3 0.3>
    componentLinearVelScale=<0.3 0.3 0.3>
    overrideWorldGravity {z=-980}
    simSpaceSettings {
     maxAngularAcceleration=200
     maxAngularVelocity=5
     maxLinearAcceleration=1000
     maxLinearVelocity=300
     velocityScaleZ=0
     worldAlpha=1
     externalLinearDragV=<1 1>
    }
   }
   n8 = RigidBody(ComponentPose=@n7, OverrideWorldGravity=@n4) {
    alphaInputType=Bool
    bClampLinearTranslationLimitToRefPose=true
    bOverrideWorldGravity=true
    overridePhysicsAsset=&/F_MED_SAM_coat_crouch_2_dyn
    simulationSpace=BaseBoneSpace
    alphaBoolBlend {blendInTime=0.1, blendOption=Linear, blendOutTime=0.1}
    baseBoneRef {boneName=pelvis}
    componentAppliedLinearAccClamp=<3000 3000 3000>
    componentLinearAccScale=<0.25 0.25 0.4>
    componentLinearVelScale=<0.2 0.2 0.2>
   }
   n9 = RigidBody!(ComponentPose=@n8, Alpha:float, OverrideWorldGravity=@n4, ExternalForce) -> (Pose) {
    bClampLinearTranslationLimitToRefPose=true
    bOverrideWorldGravity=true
    overridePhysicsAsset=&F_MED_SAM_pouches_dyn
    simulationSpace=BaseBoneSpace
    alphaBoolBlend {blendOption=Linear}
    baseBoneRef {boneName=dyn_root_simspace}
    componentAppliedLinearAccClamp=<3000 3000 3000>
    componentLinearAccScale=<0.25 0.25 0.4>
    componentLinearVelScale=<0.2 0.2 0.2>
   }
   n10 = "Transform (Modify) Bone - Bone: dyn_coat_bk_02"(ComponentPose=Local To Component(LocalPose=Use cached pose 'preCoatPose'()), Alpha=1, Translation="0,0,1", Rotation="30,0,0", Scale="1,1,1") {
    rotationMode=BMM_Additive
    rotationSpace=BCS_BoneSpace
    translationMode=BMM_Additive
    translationSpace=BCS_BoneSpace
    alphaBoolBlend {blendOption=Linear}
    boneToModify {boneName=dyn_coat_bk_02}
    rotation {pitch=30}
    translation {z=1}
   }
   n11 = "Transform (Modify) Bone - Bone: dyn_coat_l_bk_02"(ComponentPose=@n10, Alpha=1, Translation="0,1,1", Rotation="30,15,0", Scale="1,1,1") {
    rotationMode=BMM_Additive
    rotationSpace=BCS_BoneSpace
    translationMode=BMM_Additive
    translationSpace=BCS_BoneSpace
    alphaBoolBlend {blendOption=Linear}
    boneToModify {boneName=dyn_coat_l_bk_02}
    rotation {pitch=30, yaw=15}
    translation {y=1, z=1}
   }
   n12 = "Transform (Modify) Bone - Bone: dyn_coat_r_bk_02"(ComponentPose=@n11, Alpha=1, Translation="0,0,1", Rotation="30,-15,0", Scale="1,1,1") {
    rotationMode=BMM_Additive
    rotationSpace=BCS_BoneSpace
    translationMode=BMM_Additive
    translationSpace=BCS_BoneSpace
    alphaBoolBlend {blendOption=Linear}
    boneToModify {boneName=dyn_coat_r_bk_02}
    rotation {pitch=30, yaw=-15}
    translation {z=1}
   }
   n13 = Look At - Bone: dyn_coat_bk_02(ComponentPose=@n12, Alpha=1) {
    alphaBoolBlend {blendOption=Linear}
    alphaScaleBiasClamp {bInterpResult=true, interpSpeedDecreasing=15}
    boneToModify {boneName=dyn_coat_bk_02}
    lookAtLocation=<-56.4232 -12.1279 -2.9941>
    lookAtTarget {boneReference {boneName=spine_04}}
    lookAt_Axis {axis=<1 0>}
   }
   n14 = Look At - Bone: dyn_ribbon_a_01(ComponentPose=@n13, Alpha=1) {
    alphaBoolBlend {blendOption=Linear}
    alphaScaleBiasClamp {bInterpResult=true, interpSpeedDecreasing=15}
    boneToModify {boneName=dyn_ribbon_a_01}
    lookAtLocation=<48.5879 -19.6911 21.7855>
    lookAtTarget {boneReference {boneName=thigh_r}}
    lookAt_Axis {axis=<1 0>}
   }
   n15 = Look At - Bone: dyn_ribbon_b_01(ComponentPose=@n14, Alpha=1) {
    alphaBoolBlend {blendOption=Linear}
    alphaScaleBiasClamp {bInterpResult=true, interpSpeedDecreasing=15}
    boneToModify {boneName=dyn_ribbon_b_01}
    lookAtLocation=<48.5879 -33.8672 17.8182>
    lookAtTarget {boneReference {boneName=thigh_r}}
    lookAt_Axis {axis=<1 0>}
   }
   n16 = Get gravity_override()
   n17 = RigidBody(ComponentPose=@n15, OverrideWorldGravity=@n16) {
    bClampLinearTranslationLimitToRefPose=true
    bOverrideWorldGravity=true
    overridePhysicsAsset=&/F_MED_SAM_coat_backpack_dyn
    simulationSpace=BaseBoneSpace
    alphaBoolBlend {blendOption=Linear}
    baseBoneRef {boneName=pelvis}
    componentAppliedLinearAccClamp=<1000 1000 1000>
    componentLinearAccScale=<0.2 0.2 0.2>
    componentLinearVelScale=<0.2 0.2 0.2>
    overrideWorldGravity {z=-980}
   }
   n18 = RigidBody(ComponentPose=@n17, OverrideWorldGravity=@n16) {
    bClampLinearTranslationLimitToRefPose=true
    bOverrideWorldGravity=true
    overridePhysicsAsset=&F_MED_SAM_pouches_dyn
    simulationSpace=BaseBoneSpace
    alphaBoolBlend {blendOption=Linear}
    baseBoneRef {boneName=dyn_root_simspace}
    componentAppliedLinearAccClamp=<3000 3000 3000>
    componentLinearAccScale=<0.25 0.25 0.4>
    componentLinearVelScale=<0.2 0.2 0.2>
   }
   n19 = "Constraint - dyn_coat_bk_02 (P)"(ComponentPose=Local To Component(LocalPose=Use cached pose 'preCoatPose'()), Alpha=1, ConstraintWeights_0=1) {
    alphaBoolBlend {blendOption=Linear}
    boneToModify {boneName=dyn_coat_bk_02}
    constraintSetup [{targetBone {boneName=spine_03}}]
    constraintWeights [1]
   }
   n20 = Get gravity_override()
   n21 = RigidBody!(ComponentPose=@n19, bAlphaBoolEnabled:bool, OverrideWorldGravity=@n20, ExternalForce) -> (Pose) {
    alphaInputType=Bool
    bClampLinearTranslationLimitToRefPose=true
    bOverrideWorldGravity=true
    overridePhysicsAsset=&F_MED_SAM_coat_backpack_dyn
    simulationSpace=BaseBoneSpace
    alphaBoolBlend {blendInTime=0.1, blendOption=Linear, blendOutTime=0.1}
    baseBoneRef {boneName=spine_02}
    componentAppliedLinearAccClamp=<2000 2000 2000>
    componentLinearAccScale=<0.2 0.2 0.2>
    componentLinearVelScale=<0.1 0.1 0.1>
    overrideWorldGravity {z=-980}
   }
   n22 = RigidBody!(ComponentPose=@n21, bAlphaBoolEnabled, OverrideWorldGravity=Get gravity_override(), ExternalForce, bTransferBoneVelocities:bool=False) -> (Pose) {
    alphaInputType=Bool
    bClampLinearTranslationLimitToRefPose=true
    bOverrideWorldGravity=true
    overridePhysicsAsset=&F_MED_SAM_coat_crouch_2_dyn
    simulationSpace=BaseBoneSpace
    alphaBoolBlend {blendInTime=0.1, blendOption=Linear, blendOutTime=0.1}
    baseBoneRef {boneName=pelvis}
    componentAppliedLinearAccClamp=<3000 3000 3000>
    componentLinearAccScale=<0.25 0.25 0.4>
    componentLinearVelScale=<0.2 0.2 0.2>
   }
   n23 = RigidBody!(ComponentPose=@n22, Alpha:float, OverrideWorldGravity=@n20, ExternalForce) -> (Pose) {
    bClampLinearTranslationLimitToRefPose=true
    bOverrideWorldGravity=true
    overridePhysicsAsset=&F_MED_SAM_pouches_dyn
    simulationSpace=BaseBoneSpace
    alphaBoolBlend {blendOption=Linear}
    baseBoneRef {boneName=dyn_root_simspace}
    componentAppliedLinearAccClamp=<3000 3000 3000>
    componentLinearAccScale=<0.25 0.25 0.4>
    componentLinearVelScale=<0.2 0.2 0.2>
   }
   n24 = Blend Poses by bool(
    BlendPose_0=Blend Poses by bool(BlendPose_0=Component To Local(ComponentPose=@n18), BlendPose_1=Component To Local(ComponentPose=@n23), bActiveValue=Get bIsSkyDiving()) {blendType=Linear}
    BlendPose_1=Blend Poses by bool(BlendPose_0=Component To Local(ComponentPose=@n6), BlendPose_1=Component To Local(ComponentPose=@n9), bActiveValue=Get bIsSkyDiving()) {blendType=Linear}
    bActiveValue=Get wearingBackpack() -> (wearingBackpack:bool)
   ) {blendType=Linear}
   n25 = RigidBody(ComponentPose=Local To Component(LocalPose=Use cached pose 'preCoatPose'()), OverrideWorldGravity="0,0,-720") {
    bClampLinearTranslationLimitToRefPose=true
    bOverrideWorldGravity=true
    overridePhysicsAsset=&/F_MED_SAM_coat_DBNO_dyn
    simulationSpace=BaseBoneSpace
    alphaBoolBlend {blendOption=Linear}
    baseBoneRef {boneName=pelvis}
    componentAppliedLinearAccClamp=<1000 1000 1000>
    componentLinearAccScale=<0.2 0.2 0.2>
    componentLinearVelScale=<0.2 0.2 0.2>
    overrideWorldGravity {z=-980}
   }
   n26 = RigidBody(ComponentPose=@n25, OverrideWorldGravity="0,0,-980") {
    bClampLinearTranslationLimitToRefPose=true
    bOverrideWorldGravity=true
    overridePhysicsAsset=&F_MED_SAM_pouches_dyn
    simulationSpace=BaseBoneSpace
    alphaBoolBlend {blendOption=Linear}
    baseBoneRef {boneName=dyn_root_simspace}
    componentAppliedLinearAccClamp=<3000 3000 3000>
    componentLinearAccScale=<0.25 0.25 0.4>
    componentLinearVelScale=<0.2 0.2 0.2>
   }
   n27 = Component To Local(ComponentPose=@n26)
   n28 = Get bIsDBNOCarried()
   n29 = Get skydiveAlpha() -> (skydiveAlpha:double)
   n30 = "Transform (Modify) Bone - Bone: dyn_coat_l_bk_01"(
    ComponentPose=Local To Component(LocalPose=Use cached pose 'preCoatPose'())
    Alpha=@n29
    Translation="0,0,0"
    Rotation="15,0,0.000"
    Scale="1,1,1"
   ) {
    rotationMode=BMM_Additive
    rotationSpace=BCS_BoneSpace
    alphaBoolBlend {blendOption=Linear}
    alphaScaleBiasClamp {bInterpResult=true, interpSpeedDecreasing=2, interpSpeedIncreasing=5}
    boneToModify {boneName=dyn_coat_l_bk_01}
    rotation {pitch=15}
   }
   n31 = "Transform (Modify) Bone - Bone: dyn_coat_bk_01"(ComponentPose=@n30, Alpha=@n29, Translation="0,0,0", Rotation="15,0,0.000", Scale="1,1,1") {
    rotationMode=BMM_Additive
    rotationSpace=BCS_BoneSpace
    alphaBoolBlend {blendOption=Linear}
    alphaScaleBiasClamp {bInterpResult=true, interpSpeedDecreasing=2, interpSpeedIncreasing=5}
    boneToModify {boneName=dyn_coat_bk_01}
    rotation {pitch=15}
   }
   n32 = "Transform (Modify) Bone - Bone: dyn_coat_l_01"(ComponentPose=@n31, Alpha=@n29, Translation="0,0,0", Rotation="20,0,0.000", Scale="1,1,1") {
    rotationMode=BMM_Additive
    rotationSpace=BCS_BoneSpace
    alphaBoolBlend {blendOption=Linear}
    alphaScaleBiasClamp {bInterpResult=true, interpSpeedDecreasing=2, interpSpeedIncreasing=5}
    boneToModify {boneName=dyn_coat_l_01}
    rotation {pitch=20}
   }
   n33 = "Transform (Modify) Bone - Bone: dyn_coat_r_01"(ComponentPose=@n32, Alpha=@n29, Translation="0,0,0", Rotation="20,0,0.000", Scale="1,1,1") {
    rotationMode=BMM_Additive
    rotationSpace=BCS_BoneSpace
    alphaBoolBlend {blendOption=Linear}
    alphaScaleBiasClamp {bInterpResult=true, interpSpeedDecreasing=2, interpSpeedIncreasing=5}
    boneToModify {boneName=dyn_coat_r_01}
    rotation {pitch=20}
   }
   n34 = "Transform (Modify) Bone - Bone: dyn_coat_r_bk_01"(ComponentPose=@n33, Alpha=@n29, Translation="0,0,0", Rotation="15,0,0.000", Scale="1,1,1") {
    rotationMode=BMM_Additive
    rotationSpace=BCS_BoneSpace
    alphaBoolBlend {blendOption=Linear}
    alphaScaleBiasClamp {bInterpResult=true, interpSpeedDecreasing=2, interpSpeedIncreasing=5}
    boneToModify {boneName=dyn_coat_r_bk_01}
    rotation {pitch=15}
   }
   n35 = Get ParachuteAlpha() -> (ParachuteAlpha:double)
   n36 = "Transform (Modify) Bone - Bone: dyn_coat_l_bk_01"(ComponentPose=@n34, Alpha=@n35, Translation="0,0,0", Rotation="5,0,0.000", Scale="1,1,1") {
    rotationMode=BMM_Additive
    rotationSpace=BCS_BoneSpace
    alphaBoolBlend {blendOption=Linear}
    alphaScaleBiasClamp {bInterpResult=true, interpSpeedDecreasing=2, interpSpeedIncreasing=5}
    boneToModify {boneName=dyn_coat_l_bk_01}
    rotation {pitch=5}
   }
   n37 = "Transform (Modify) Bone - Bone: dyn_coat_bk_01"(ComponentPose=@n36, Alpha=@n35, Translation="0,0,0", Rotation="5,0,0.000", Scale="1,1,1") {
    rotationMode=BMM_Additive
    rotationSpace=BCS_BoneSpace
    alphaBoolBlend {blendOption=Linear}
    alphaScaleBiasClamp {bInterpResult=true, interpSpeedDecreasing=2, interpSpeedIncreasing=5}
    boneToModify {boneName=dyn_coat_bk_01}
    rotation {pitch=5}
   }
   n38 = "Transform (Modify) Bone - Bone: dyn_coat_l_01"(ComponentPose=@n37, Alpha=@n35, Translation="0,0,0", Rotation="15,0,0.000", Scale="1,1,1") {
    rotationMode=BMM_Additive
    rotationSpace=BCS_BoneSpace
    alphaBoolBlend {blendOption=Linear}
    alphaScaleBiasClamp {bInterpResult=true, interpSpeedDecreasing=2, interpSpeedIncreasing=5}
    boneToModify {boneName=dyn_coat_l_01}
    rotation {pitch=15}
   }
   n39 = "Transform (Modify) Bone - Bone: dyn_coat_r_01"(ComponentPose=@n38, Alpha=@n35, Translation="0,0,0", Rotation="10,0,0.000", Scale="1,1,1") {
    rotationMode=BMM_Additive
    rotationSpace=BCS_BoneSpace
    alphaBoolBlend {blendOption=Linear}
    alphaScaleBiasClamp {bInterpResult=true, interpSpeedDecreasing=2, interpSpeedIncreasing=5}
    boneToModify {boneName=dyn_coat_r_01}
    rotation {pitch=10}
   }
   n40 = "Transform (Modify) Bone - Bone: dyn_coat_r_bk_01"(ComponentPose=@n39, Alpha=@n35, Translation="0,0,0", Rotation="5,0,0.000", Scale="1,1,1") {
    rotationMode=BMM_Additive
    rotationSpace=BCS_BoneSpace
    alphaBoolBlend {blendOption=Linear}
    alphaScaleBiasClamp {bInterpResult=true, interpSpeedDecreasing=2, interpSpeedIncreasing=5}
    boneToModify {boneName=dyn_coat_r_bk_01}
    rotation {pitch=5}
   }
   n41 = Get BwdMovement() -> (BwdMovement:double)
   n42 = Trail controller - Bone: dyn_coat_r_bk_05!(ComponentPose=@n40, Alpha=@n41, FakeVelocity:Vector="0,0,0") -> (Pose) {
    bActorSpaceFakeVel=true
    bShowTrailLocation=true
    chainLength=4
    debugLifeTime=0.1
    maxDeltaTime=0.0333
    relaxationSpeedScale=3
    alphaScaleBiasClamp {bInterpResult=true}
    baseJoint {boneName=root}
    planarLimits [
     {drivingBone {boneName=root}, planeTransform {location=<0 0 10>, rotation=<rot 0 0 0>, scale=<1 1 1>}}
     {drivingBone {boneName=spine_03}, planeTransform {location=<0 -20 0>, rotation=<rot 0 15 -89.9999>, scale=<1 1 1>}}
    ]
    rotationLimits [{limitMax=<180 180 180>}, {limitMax=<180 180 180>}, {limitMax=<180 180 180>}, {limitMax=<180 180 180>}]
    rotationOffsets [<0 0 0>, <0 0 0>, <0 0 0>, <0 0 0>]
    trailBone {boneName=dyn_coat_r_bk_05}
    trailRelaxationSpeed {editorCurveData {keys [{arriveTangent=-72.0739, interpMode=RCIM_Cubic, leaveTangent=-72.0739, tangentMode=RCTM_User, time=0, value=35}, <at 1 5 RCIM_Cubic>]}}
   }
   n43 = Trail controller - Bone: dyn_coat_l_bk_05!(ComponentPose=@n42, Alpha=@n41, FakeVelocity:Vector="0,0,0") -> (Pose) {
    bActorSpaceFakeVel=true
    bShowTrailLocation=true
    chainLength=4
    debugLifeTime=0.1
    maxDeltaTime=0.0333
    relaxationSpeedScale=3
    alphaScaleBiasClamp {bInterpResult=true}
    baseJoint {boneName=root}
    planarLimits [
     {drivingBone {boneName=root}, planeTransform {location=<0 0 10>, rotation=<rot 0 0 0>, scale=<1 1 1>}}
     {drivingBone {boneName=spine_03}, planeTransform {location=<0 -20 0>, rotation=<rot 0 15 -89.9999>, scale=<1 1 1>}}
    ]
    rotationLimits [{limitMax=<180 180 180>}, {limitMax=<180 180 180>}, {limitMax=<180 180 180>}, {limitMax=<180 180 180>}]
    rotationOffsets [<0 0 0>, <0 0 0>, <0 0 0>, <0 0 0>]
    trailBone {boneName=dyn_coat_l_bk_05}
    trailRelaxationSpeed {editorCurveData {keys [{arriveTangent=-72.0739, interpMode=RCIM_Cubic, leaveTangent=-72.0739, tangentMode=RCTM_User, time=0, value=35}, <at 1 5 RCIM_Cubic>]}}
   }
   n44 = Trail controller - Bone: dyn_coat_bk_05!(ComponentPose=@n43, Alpha=@n41, FakeVelocity:Vector="0,0,0") -> (Pose) {
    bActorSpaceFakeVel=true
    bShowTrailLocation=true
    chainLength=4
    debugLifeTime=0.1
    maxDeltaTime=0.0333
    relaxationSpeedScale=3
    alphaScaleBiasClamp {bInterpResult=true}
    baseJoint {boneName=root}
    planarLimits [
     {drivingBone {boneName=root}, planeTransform {location=<0 0 10>, rotation=<rot 0 0 0>, scale=<1 1 1>}}
     {drivingBone {boneName=spine_03}, planeTransform {location=<0 -20 0>, rotation=<rot 0 15 -89.9999>, scale=<1 1 1>}}
    ]
    rotationLimits [{limitMax=<180 180 180>}, {limitMax=<180 180 180>}, {limitMax=<180 180 180>}, {limitMax=<180 180 180>}]
    rotationOffsets [<0 0 0>, <0 0 0>, <0 0 0>, <0 0 0>]
    trailBone {boneName=dyn_coat_bk_05}
    trailRelaxationSpeed {editorCurveData {keys [{arriveTangent=-72.0739, interpMode=RCIM_Cubic, leaveTangent=-72.0739, tangentMode=RCTM_User, time=0, value=35}, <at 1 5 RCIM_Cubic>]}}
   }
   n45 = Trail controller - Bone: dyn_coat_r_06!(ComponentPose=@n44, Alpha=Get LeftMovement() -> (LeftMovement:double), FakeVelocity:Vector="0,0,0") -> (Pose) {
    bActorSpaceFakeVel=true
    bShowTrailLocation=true
    chainLength=4
    debugLifeTime=0.1
    maxDeltaTime=0.0333
    relaxationSpeedScale=3
    alphaScaleBiasClamp {bInterpResult=true}
    baseJoint {boneName=root}
    planarLimits [
     {drivingBone {boneName=root}, planeTransform {location=<0 0 10>, rotation=<rot 0 0 0>, scale=<1 1 1>}}
     {drivingBone {boneName=spine_03}, planeTransform {location=<0 -20 0>, rotation=<rot 0 15 -89.9999>, scale=<1 1 1>}}
    ]
    rotationLimits [{limitMax=<180 180 180>}, {limitMax=<180 180 180>}, {limitMax=<180 180 180>}, {limitMax=<180 180 180>}]
    rotationOffsets [<0 0 0>, <0 0 0>, <0 0 0>, <0 0 0>]
    trailBone {boneName=dyn_coat_r_06}
    trailRelaxationSpeed {editorCurveData {keys [{arriveTangent=-72.0739, interpMode=RCIM_Cubic, leaveTangent=-72.0739, tangentMode=RCTM_User, time=0, value=35}, <at 1 5 RCIM_Cubic>]}}
   }
   n46 = Trail controller - Bone: dyn_coat_l_06!(ComponentPose=@n45, Alpha=Get RightMovement() -> (RightMovement:double), FakeVelocity:Vector="0,0,0") -> (Pose) {
    bActorSpaceFakeVel=true
    bShowTrailLocation=true
    chainLength=4
    debugLifeTime=0.1
    maxDeltaTime=0.0333
    relaxationSpeedScale=3
    alphaScaleBiasClamp {bInterpResult=true}
    baseJoint {boneName=root}
    planarLimits [
     {drivingBone {boneName=root}, planeTransform {location=<0 0 10>, rotation=<rot 0 0 0>, scale=<1 1 1>}}
     {drivingBone {boneName=spine_03}, planeTransform {location=<0 -20 0>, rotation=<rot 0 15 -89.9999>, scale=<1 1 1>}}
    ]
    rotationLimits [{limitMax=<180 180 180>}, {limitMax=<180 180 180>}, {limitMax=<180 180 180>}, {limitMax=<180 180 180>}]
    rotationOffsets [<0 0 0>, <0 0 0>, <0 0 0>, <0 0 0>]
    trailBone {boneName=dyn_coat_l_06}
    trailRelaxationSpeed {editorCurveData {keys [{arriveTangent=-72.0739, interpMode=RCIM_Cubic, leaveTangent=-72.0739, tangentMode=RCTM_User, time=0, value=35}, <at 1 5 RCIM_Cubic>]}}
   }
   Output Pose(Result=Blend Poses by bool(
     BlendPose_0=Blend Poses by bool(BlendPose_0=@n27, BlendPose_1=@n24, bActiveValue=@n28) {blendType=Linear}
     BlendPose_1=Component To Local(ComponentPose=@n46)
     bActiveValue=Get bRigidBodyEnabled()
    ) {blendType=Linear}) {layerGroup=None}
   n47 = "Constraint - dyn_bowknot_MSTR (PPPP)"!(
    ComponentPose=Local To Component(LocalPose=Copy Pose From Mesh() {bUseAttachedParent=true})
    Alpha=1
    ConstraintWeights_0=0
    ConstraintWeights_1:float=0.0618
    ConstraintWeights_2:float=0.089
    ConstraintWeights_3:float=0.8491
   ) -> (Pose) {
    alphaBoolBlend {blendOption=Linear}
    boneToModify {boneName=dyn_bowknot_MSTR}
    constraintSetup [{targetBone {boneName=thigh_r}}, {targetBone {boneName=spine_01}}, {targetBone {boneName=thigh_twist_01_r}}, {targetBone {boneName=pelvis}}]
    constraintWeights [1,1,1,1]
   }
   n48 = "Constraint - dyn_pouch_base_MSTR (PPPP)"!(
    ComponentPose=@n47
    Alpha=1
    ConstraintWeights_0=0.0034
    ConstraintWeights_1:float=0.4397
    ConstraintWeights_2:float=0.0207
    ConstraintWeights_3:float=0.5362
   ) -> (Pose) {
    alphaBoolBlend {blendOption=Linear}
    boneToModify {boneName=dyn_pouch_base_MSTR}
    constraintSetup [{targetBone {boneName=spine_01}}, {targetBone {boneName=thigh_twist_01_l}}, {targetBone {boneName=thigh_l}}, {targetBone {boneName=pelvis}}]
    constraintWeights [1,1,1,1]
   }
   n49 = "Constraint - dyn_pouch_base_MSTR (PPP)"!(ComponentPose=@n48, Alpha=1, ConstraintWeights_0=0.0182, ConstraintWeights_1:float=0.6089, ConstraintWeights_2:float=0.3729) -> (Pose) {
    alphaBoolBlend {blendOption=Linear}
    boneToModify {boneName=dyn_pouch_base_MSTR}
    constraintSetup [{targetBone {boneName=pelvis}}, {targetBone {boneName=thigh_twist_01_l}}, {targetBone {boneName=thigh_l}}]
    constraintWeights [1,1,1]
   }
   n50 = "Constraint - dyn_pouch_beads_02_MSTR (PPPP)"!(
    ComponentPose=@n49
    Alpha=1
    ConstraintWeights_0=0.0217
    ConstraintWeights_1:float=0.364
    ConstraintWeights_2:float=0.0164
    ConstraintWeights_3:float=0.5978
   ) -> (Pose) {
    alphaBoolBlend {blendOption=Linear}
    boneToModify {boneName=dyn_pouch_beads_02_MSTR}
    constraintSetup [{targetBone {boneName=spine_01}}, {targetBone {boneName=thigh_twist_01_l}}, {targetBone {boneName=thigh_l}}, {targetBone {boneName=pelvis}}]
    constraintWeights [1,1,1,1]
   }
   n51 = "Constraint - dyn_bag_MSTR (PPP)"!(ComponentPose=@n50, Alpha=1, ConstraintWeights_0=0.0025, ConstraintWeights_1:float=0.5406, ConstraintWeights_2:float=0.4568) -> (Pose) {
    alphaBoolBlend {blendOption=Linear}
    boneToModify {boneName=dyn_bag_MSTR}
    constraintSetup [{targetBone {boneName=pelvis}}, {targetBone {boneName=thigh_twist_01_l}}, {targetBone {boneName=thigh_l}}]
    constraintWeights [1,1,1]
   }
   n52 = "Constraint - dyn_bracelet_l_01_MSTR (PPP)"!(ComponentPose=@n51, Alpha=1, ConstraintWeights_0=0.1022, ConstraintWeights_1:float=0.8927, ConstraintWeights_2:float=0.0051) -> (Pose) {
    alphaBoolBlend {blendOption=Linear}
    boneToModify {boneName=dyn_bracelet_l_01_MSTR}
    constraintSetup [{targetBone {boneName=hand_l}}, {targetBone {boneName=lowerarm_twist_01_l}}, {targetBone {boneName=lowerarm_twist_02_l}}]
    constraintWeights [1,1,1]
   }
   n53 = "Constraint - dyn_bracelet_l_02_MSTR (PPP)"!(ComponentPose=@n52, Alpha=1, ConstraintWeights_0=0.1022, ConstraintWeights_1:float=0.8927, ConstraintWeights_2:float=0.0051) -> (Pose) {
    alphaBoolBlend {blendOption=Linear}
    boneToModify {boneName=dyn_bracelet_l_02_MSTR}
    constraintSetup [{targetBone {boneName=hand_l}}, {targetBone {boneName=lowerarm_twist_01_l}}, {targetBone {boneName=lowerarm_twist_02_l}}]
    constraintWeights [1,1,1]
   }
   n54 = "Constraint - dyn_bracelet_r_01_MSTR (PPP)"!(ComponentPose=@n53, Alpha=1, ConstraintWeights_0=0.0051, ConstraintWeights_1:float=0.1022, ConstraintWeights_2:float=0.8927) -> (Pose) {
    alphaBoolBlend {blendOption=Linear}
    boneToModify {boneName=dyn_bracelet_r_01_MSTR}
    constraintSetup [{targetBone {boneName=lowerarm_twist_02_r}}, {targetBone {boneName=hand_r}}, {targetBone {boneName=lowerarm_twist_01_r}}]
    constraintWeights [1,1,1]
   }
   n55 = "Constraint - dyn_bracelet_r_02_MSTR (PPP)"!(ComponentPose=@n54, Alpha=1, ConstraintWeights_0=0.0051, ConstraintWeights_1:float=0.1022, ConstraintWeights_2:float=0.8927) -> (Pose) {
    alphaBoolBlend {blendOption=Linear}
    boneToModify {boneName=dyn_bracelet_r_02_MSTR}
    constraintSetup [{targetBone {boneName=lowerarm_twist_02_r}}, {targetBone {boneName=hand_r}}, {targetBone {boneName=lowerarm_twist_01_r}}]
    constraintWeights [1,1,1]
   }
   n56 = "Transform (Modify) Bone - Bone: dyn_bracelet_l_01"(ComponentPose=@n55, Alpha=1, Translation="-0.75,0,0", Rotation="0,0,0", Scale="1,1,1") {
    rotationMode=BMM_Additive
    translationMode=BMM_Additive
    translationSpace=BCS_BoneSpace
    alphaBoolBlend {blendOption=Linear}
    boneToModify {boneName=dyn_bracelet_l_01}
    translation {x=-0.75}
   }
   n57 = "Transform (Modify) Bone - Bone: dyn_bracelet_l_02"(ComponentPose=@n56, Alpha=1, Translation="-0.75,0,0", Rotation="0,0,0", Scale="1,1,1") {
    rotationMode=BMM_Additive
    rotationSpace=BCS_BoneSpace
    translationMode=BMM_Additive
    translationSpace=BCS_BoneSpace
    alphaBoolBlend {blendOption=Linear}
    boneToModify {boneName=dyn_bracelet_l_02}
    translation {x=-0.75}
   }
   n58 = "Transform (Modify) Bone - Bone: dyn_bracelet_r_01"(ComponentPose=@n57, Alpha=1, Translation="-0.75,0,0", Rotation="0,0,0", Scale="1,1,1") {
    rotationMode=BMM_Additive
    translationMode=BMM_Additive
    translationSpace=BCS_BoneSpace
    alphaBoolBlend {blendOption=Linear}
    boneToModify {boneName=dyn_bracelet_r_01}
    translation {x=-0.75}
   }
   n59 = "Transform (Modify) Bone - Bone: dyn_bracelet_r_02"(ComponentPose=@n58, Alpha=1, Translation="-0.75,0,0", Rotation="0,0,0", Scale="1,1,1") {
    rotationMode=BMM_Additive
    rotationSpace=BCS_BoneSpace
    translationMode=BMM_Additive
    translationSpace=BCS_BoneSpace
    alphaBoolBlend {blendOption=Linear}
    boneToModify {boneName=dyn_bracelet_r_02}
    translation {x=-0.75}
   }
   n60 = "Constraint - dyn_bracelet_l_02 (R)"(ComponentPose=@n59, Alpha=0.8, ConstraintWeights_0=1) {
    alphaBoolBlend {blendOption=Linear}
    boneToModify {boneName=dyn_bracelet_l_02}
    constraintSetup [{transformType=Rotation, targetBone {boneName=hand_l}}]
    constraintWeights [1]
   }
   n61 = "Constraint - dyn_bracelet_l_01 (R)"(ComponentPose=@n60, Alpha=0.8, ConstraintWeights_0=1) {
    alphaBoolBlend {blendOption=Linear}
    boneToModify {boneName=dyn_bracelet_l_01}
    constraintSetup [{transformType=Rotation, targetBone {boneName=dyn_bracelet_l_02}}]
    constraintWeights [1]
   }
   n62 = "Constraint - dyn_bracelet_r_02 (R)"(ComponentPose=@n61, Alpha=0.8, ConstraintWeights_0=1) {
    alphaBoolBlend {blendOption=Linear}
    boneToModify {boneName=dyn_bracelet_r_02}
    constraintSetup [{transformType=Rotation, targetBone {boneName=hand_r}}]
    constraintWeights [1]
   }
   n63 = "Constraint - dyn_bracelet_r_01 (R)"(ComponentPose=@n62, Alpha=0.8, ConstraintWeights_0=1) {
    alphaBoolBlend {blendOption=Linear}
    boneToModify {boneName=dyn_bracelet_r_01}
    constraintSetup [{transformType=Rotation, targetBone {boneName=dyn_bracelet_r_02}}]
    constraintWeights [1]
   }
   n64 = Anim Dynamics - Bone: dyn_bowknot_01(ComponentPose=@n63, Alpha=1, ExternalForce="0,0,0") {
    angularDampingOverride=0.9999
    angularSpringConstant=100
    bAngularSpring=true
    bGravityOverrideInSimSpace=true
    bOverrideAngularDamping=true
    simulationSpace=RootRelative
    alphaBoolBlend {blendOption=Linear}
    boundBone {boneName=dyn_bowknot_01}
    componentAppliedLinearAccClamp=<3500 3500 3500>
    componentLinearAccScale=<1.2 1.2 1.2>
    componentLinearVelScale=<1.2 1.2 1.2>
    physicsBodyDefinitions [{
      sphereCollisionRadius=0
      boundBone {boneName=dyn_bowknot_01}
      boxExtents=<20 10 10>
      constraintSetup {angularLimitsMax=<1 10 15>, angularLimitsMin=<-1 -25 -15>, angularTarget=<0 0 0>}
      localJointOffset=<1 0 0>
     }]
    relativeSpaceBone {boneName=root}
   }
   n65 = Anim Dynamics - Bone: dyn_bowknot_02(ComponentPose=@n64, Alpha=1, ExternalForce="0,0,0") {
    angularDampingOverride=0.9999
    angularSpringConstant=100
    bAngularSpring=true
    bGravityOverrideInSimSpace=true
    bOverrideAngularDamping=true
    simulationSpace=RootRelative
    alphaBoolBlend {blendOption=Linear}
    boundBone {boneName=dyn_bowknot_02}
    componentAppliedLinearAccClamp=<3500 3500 3500>
    componentLinearAccScale=<1.2 1.2 1.2>
    componentLinearVelScale=<1.2 1.2 1.2>
    physicsBodyDefinitions [{
      sphereCollisionRadius=0
      boundBone {boneName=dyn_bowknot_02}
      boxExtents=<20 10 10>
      constraintSetup {angularLimitsMax=<1 10 15>, angularLimitsMin=<-1 -25 -15>, angularTarget=<0 0 0>}
      localJointOffset=<1 0 0>
     }]
    relativeSpaceBone {boneName=root}
   }
   n66 = Anim Dynamics - Bone: dyn_flap_ribbon_a_01(ComponentPose=@n65, Alpha=1, ExternalForce="0,0,0") {
    angularDampingOverride=0.9999
    angularSpringConstant=100
    bAngularSpring=true
    bGravityOverrideInSimSpace=true
    bOverrideAngularDamping=true
    simulationSpace=RootRelative
    alphaBoolBlend {blendOption=Linear}
    boundBone {boneName=dyn_flap_ribbon_a_01}
    componentAppliedLinearAccClamp=<3500 3500 3500>
    componentLinearAccScale=<1.2 1.2 1.2>
    componentLinearVelScale=<1.2 1.2 1.2>
    physicsBodyDefinitions [{
      sphereCollisionRadius=0
      boundBone {boneName=dyn_flap_ribbon_a_01}
      boxExtents=<20 10 10>
      constraintSetup {angularLimitsMax=<1 0 15>, angularLimitsMin=<-1 -10 -15>, angularTarget=<0 0 0>}
      localJointOffset=<2 0 0>
     }]
    relativeSpaceBone {boneName=root}
   }
   n67 = Anim Dynamics - Bone: dyn_flap_ribbon_b_01(ComponentPose=@n66, Alpha=1, ExternalForce="0,0,0") {
    angularDampingOverride=0.9999
    angularSpringConstant=50
    bAngularSpring=true
    bGravityOverrideInSimSpace=true
    bOverrideAngularDamping=true
    simulationSpace=RootRelative
    alphaBoolBlend {blendOption=Linear}
    boundBone {boneName=dyn_flap_ribbon_b_01}
    componentAppliedLinearAccClamp=<3500 3500 3500>
    componentLinearAccScale=<1.2 1.2 1.2>
    componentLinearVelScale=<1.2 1.2 1.2>
    physicsBodyDefinitions [{
      sphereCollisionRadius=0
      boundBone {boneName=dyn_flap_ribbon_b_01}
      boxExtents=<20 10 10>
      constraintSetup {angularLimitsMax=<1 2 15>, angularLimitsMin=<-1 -10 -15>, angularTarget=<0 0 0>}
      localJointOffset=<1 0 0>
     }]
    relativeSpaceBone {boneName=root}
   }
   n68 = Anim Dynamics - Bone: dyn_flap_ribbon_a_02(ComponentPose=@n67, Alpha=1, ExternalForce="0,0,0") {
    angularDampingOverride=0.9999
    angularSpringConstant=100
    bAngularSpring=true
    bGravityOverrideInSimSpace=true
    bOverrideAngularDamping=true
    simulationSpace=RootRelative
    alphaBoolBlend {blendOption=Linear}
    boundBone {boneName=dyn_flap_ribbon_a_02}
    componentAppliedLinearAccClamp=<3500 3500 3500>
    componentLinearAccScale=<1.2 1.2 1.2>
    componentLinearVelScale=<1.2 1.2 1.2>
    physicsBodyDefinitions [{
      sphereCollisionRadius=0
      boundBone {boneName=dyn_flap_ribbon_a_02}
      boxExtents=<15 7 7>
      constraintSetup {angularLimitsMax=<1 10 10>, angularLimitsMin=<-1 -15 -10>, angularTarget=<0 0 0>}
      localJointOffset=<2 0 0>
     }]
    relativeSpaceBone {boneName=root}
   }
   n69 = Anim Dynamics - Bone: dyn_flap_ribbon_b_02(ComponentPose=@n68, Alpha=1, ExternalForce="0,0,0") {
    angularDampingOverride=0.9999
    angularSpringConstant=50
    bAngularSpring=true
    bGravityOverrideInSimSpace=true
    bOverrideAngularDamping=true
    simulationSpace=RootRelative
    alphaBoolBlend {blendOption=Linear}
    boundBone {boneName=dyn_flap_ribbon_b_02}
    componentAppliedLinearAccClamp=<3500 3500 3500>
    componentLinearAccScale=<1.2 1.2 1.2>
    componentLinearVelScale=<1.2 1.2 1.2>
    physicsBodyDefinitions [{
      sphereCollisionRadius=0
      boundBone {boneName=dyn_flap_ribbon_b_02}
      boxExtents=<15 7 7>
      constraintSetup {angularLimitsMax=<1 10 10>, angularLimitsMin=<-1 -15 -10>, angularTarget=<0 0 0>}
      localJointOffset=<2 0 0>
     }]
    relativeSpaceBone {boneName=root}
   }
   n70 = Anim Dynamics - Bone: dyn_cuff_l_01(ComponentPose=@n69, Alpha=1, ExternalForce="0,0,0") {
    angularDampingOverride=0.9999
    angularSpringConstant=25
    bAngularSpring=true
    bGravityOverrideInSimSpace=true
    bOverrideAngularDamping=true
    simulationSpace=RootRelative
    alphaBoolBlend {blendOption=Linear}
    boundBone {boneName=dyn_cuff_l_01}
    componentAppliedLinearAccClamp=<3500 3500 3500>
    componentLinearAccScale=<0.8 0.8 0.8>
    componentLinearVelScale=<0.8 0.8 0.8>
    physicsBodyDefinitions [{
      sphereCollisionRadius=0
      boundBone {boneName=dyn_cuff_l_01}
      boxExtents=<20 10 10>
      constraintSetup {angularLimitsMax=<1 5 5>, angularLimitsMin=<-1 -15 -5>, angularTarget=<0 0 0>}
      localJointOffset=<2 0 0>
     }]
    relativeSpaceBone {boneName=pelvis}
   }
   n71 = Anim Dynamics - Bone: dyn_cuff_l_02(ComponentPose=@n70, Alpha=1, ExternalForce="0,0,0") {
    angularDampingOverride=0.9999
    angularSpringConstant=25
    bAngularSpring=true
    bGravityOverrideInSimSpace=true
    bOverrideAngularDamping=true
    simulationSpace=RootRelative
    alphaBoolBlend {blendOption=Linear}
    boundBone {boneName=dyn_cuff_l_02}
    componentAppliedLinearAccClamp=<3500 3500 3500>
    componentLinearAccScale=<0.8 0.8 0.8>
    componentLinearVelScale=<0.8 0.8 0.8>
    physicsBodyDefinitions [{
      sphereCollisionRadius=0
      boundBone {boneName=dyn_cuff_l_02}
      boxExtents=<20 10 10>
      constraintSetup {angularLimitsMax=<1 5 5>, angularLimitsMin=<-1 -15 -5>, angularTarget=<0 0 0>}
      localJointOffset=<2 0 0>
     }]
    relativeSpaceBone {boneName=pelvis}
   }
   n72 = Anim Dynamics - Bone: dyn_cuff_l_03(ComponentPose=@n71, Alpha=1, ExternalForce="0,0,0") {
    angularDampingOverride=0.9999
    angularSpringConstant=25
    bAngularSpring=true
    bGravityOverrideInSimSpace=true
    bOverrideAngularDamping=true
    simulationSpace=RootRelative
    alphaBoolBlend {blendOption=Linear}
    boundBone {boneName=dyn_cuff_l_03}
    componentAppliedLinearAccClamp=<3500 3500 3500>
    componentLinearAccScale=<0.8 0.8 0.8>
    componentLinearVelScale=<0.8 0.8 0.8>
    physicsBodyDefinitions [{
      sphereCollisionRadius=0
      boundBone {boneName=dyn_cuff_l_03}
      boxExtents=<20 10 10>
      constraintSetup {angularLimitsMax=<1 5 5>, angularLimitsMin=<-1 -15 -5>, angularTarget=<0 0 0>}
      localJointOffset=<2 0 0>
     }]
    relativeSpaceBone {boneName=pelvis}
   }
   n73 = Anim Dynamics - Bone: dyn_cuff_l_04(ComponentPose=@n72, Alpha=1, ExternalForce="0,0,0") {
    angularDampingOverride=0.9999
    angularSpringConstant=25
    bAngularSpring=true
    bGravityOverrideInSimSpace=true
    bOverrideAngularDamping=true
    simulationSpace=RootRelative
    alphaBoolBlend {blendOption=Linear}
    boundBone {boneName=dyn_cuff_l_04}
    componentAppliedLinearAccClamp=<3500 3500 3500>
    componentLinearAccScale=<0.8 0.8 0.8>
    componentLinearVelScale=<0.8 0.8 0.8>
    physicsBodyDefinitions [{
      sphereCollisionRadius=0
      boundBone {boneName=dyn_cuff_l_04}
      boxExtents=<20 10 10>
      constraintSetup {angularLimitsMax=<1 5 5>, angularLimitsMin=<-1 -15 -5>, angularTarget=<0 0 0>}
      localJointOffset=<2 0 0>
     }]
    relativeSpaceBone {boneName=pelvis}
   }
   n74 = Anim Dynamics - Bone: dyn_cuff_r_01(ComponentPose=@n73, Alpha=1, ExternalForce="0,0,0") {
    angularDampingOverride=0.9999
    angularSpringConstant=25
    bAngularSpring=true
    bGravityOverrideInSimSpace=true
    bOverrideAngularDamping=true
    simulationSpace=RootRelative
    alphaBoolBlend {blendOption=Linear}
    boundBone {boneName=dyn_cuff_r_01}
    componentAppliedLinearAccClamp=<3500 3500 3500>
    componentLinearAccScale=<0.8 0.8 0.8>
    componentLinearVelScale=<0.8 0.8 0.8>
    physicsBodyDefinitions [{
      sphereCollisionRadius=0
      boundBone {boneName=dyn_cuff_r_01}
      boxExtents=<20 10 10>
      constraintSetup {angularLimitsMax=<1 5 5>, angularLimitsMin=<-1 -15 -5>, angularTarget=<0 0 0>}
      localJointOffset=<2 0 0>
     }]
    relativeSpaceBone {boneName=pelvis}
   }
   n75 = Anim Dynamics - Bone: dyn_cuff_r_02(ComponentPose=@n74, Alpha=1, ExternalForce="0,0,0") {
    angularDampingOverride=0.9999
    angularSpringConstant=25
    bAngularSpring=true
    bGravityOverrideInSimSpace=true
    bOverrideAngularDamping=true
    simulationSpace=RootRelative
    alphaBoolBlend {blendOption=Linear}
    boundBone {boneName=dyn_cuff_r_02}
    componentAppliedLinearAccClamp=<3500 3500 3500>
    componentLinearAccScale=<0.8 0.8 0.8>
    componentLinearVelScale=<0.8 0.8 0.8>
    physicsBodyDefinitions [{
      sphereCollisionRadius=0
      boundBone {boneName=dyn_cuff_r_02}
      boxExtents=<20 10 10>
      constraintSetup {angularLimitsMax=<1 5 5>, angularLimitsMin=<-1 -15 -5>, angularTarget=<0 0 0>}
      localJointOffset=<2 0 0>
     }]
    relativeSpaceBone {boneName=pelvis}
   }
   n76 = Anim Dynamics - Bone: dyn_cuff_r_03(ComponentPose=@n75, Alpha=1, ExternalForce="0,0,0") {
    angularDampingOverride=0.9999
    angularSpringConstant=25
    bAngularSpring=true
    bGravityOverrideInSimSpace=true
    bOverrideAngularDamping=true
    simulationSpace=RootRelative
    alphaBoolBlend {blendOption=Linear}
    boundBone {boneName=dyn_cuff_r_03}
    componentAppliedLinearAccClamp=<3500 3500 3500>
    componentLinearAccScale=<0.8 0.8 0.8>
    componentLinearVelScale=<0.8 0.8 0.8>
    physicsBodyDefinitions [{
      sphereCollisionRadius=0
      boundBone {boneName=dyn_cuff_r_03}
      boxExtents=<20 10 10>
      constraintSetup {angularLimitsMax=<1 5 5>, angularLimitsMin=<-1 -15 -5>, angularTarget=<0 0 0>}
      localJointOffset=<2 0 0>
     }]
    relativeSpaceBone {boneName=pelvis}
   }
   n77 = Anim Dynamics - Bone: dyn_cuff_r_04(ComponentPose=@n76, Alpha=1, ExternalForce="0,0,0") {
    angularDampingOverride=0.9999
    angularSpringConstant=25
    bAngularSpring=true
    bGravityOverrideInSimSpace=true
    bOverrideAngularDamping=true
    simulationSpace=RootRelative
    alphaBoolBlend {blendOption=Linear}
    boundBone {boneName=dyn_cuff_r_04}
    componentAppliedLinearAccClamp=<3500 3500 3500>
    componentLinearAccScale=<0.8 0.8 0.8>
    componentLinearVelScale=<0.8 0.8 0.8>
    physicsBodyDefinitions [{
      sphereCollisionRadius=0
      boundBone {boneName=dyn_cuff_r_04}
      boxExtents=<20 10 10>
      constraintSetup {angularLimitsMax=<1 5 5>, angularLimitsMin=<-1 -15 -5>, angularTarget=<0 0 0>}
      localJointOffset=<2 0 0>
     }]
    relativeSpaceBone {boneName=pelvis}
   }
   Save cached pose 'preCoatPose'(Pose=Component To Local(ComponentPose=@n77)) {cacheName=preCoatPose}
  ]}
 EventGraph {nodes [
   Event BlueprintUpdateAnimation()
   > n1 = Sequence!(execute) -> (then_0, then_1, then_2:Exec, then_3:Exec, then_4:Exec, then_5:Exec, then_6:Exec, then_7:Exec)
   > n2 = Gravity Override State Function DEPRECATED(CharacterPartAnimInstance=Self-Reference(), GameTimeInSeconds=GetGameTimeInSeconds(), GravityOverrideParams=Get Gravity Override Params())
   > Set gravity_override(gravity_override=@n2)
   n3 = ComplexSimSpaceFunction DEPRECATED!(
    execute=@n1.then_1
    CharacterPartAnimInstance:Custom Character Part Anim Instance Object Reference=Self-Reference()
    InterpSpeedMax:double=9
    PawnSpeedMax:double=20
    RInterpErrorTolerance:double=1
    DeltaTime:double=0
    DebugDraw:bool=false
    sim_space_rotation_in:Rotator=Get sim_space_rotation()
    inComponentSpaceIn:bool=Get InComponentSpace() -> (InComponentSpace:bool)
    inComponentSpaceLastFrameIn:bool=Get InComponentSpaceLastFrame() -> (InComponentSpaceLastFrame:bool)
    lockedOnSimSpaceIn:bool=Get LockedOnSimSpace() -> (LockedOnSimSpace:bool)
    InterpSpeedIn:double=Get InterpSpeed() -> (InterpSpeed:double)
   ) -> (
    then
    sim_space_rotation:Rotator
    inComponentSpaceOut:bool
    inComponentSpaceLastFrameOut:bool
    lockedOnSimSpaceOut:bool
    InterpSpeedOut:double
   )
   > Set sim_space_rotation(sim_space_rotation:Rotator=@n3) -> (Output_Get:Rotator)
   > Set InComponentSpace(InComponentSpace:bool=@n3.inComponentSpaceOut) -> (Output_Get:bool)
   > Set InComponentSpaceLastFrame(InComponentSpaceLastFrame:bool=@n3.inComponentSpaceLastFrameOut) -> (Output_Get:bool)
   > Set LockedOnSimSpace(LockedOnSimSpace:bool=@n3.lockedOnSimSpaceOut) -> (Output_Get:bool)
   > Set InterpSpeed(InterpSpeed:double=@n3.InterpSpeedOut) -> (Output_Get:double)
   Set hand l rot(
    execute=@n1.then_2
    hand l rot:double=GetAlphaFromReferenceTransformDelta(SkeletalMeshComponent=Get Skel_Mesh(), BoneName=hand_l, BaseName=lowerarm_l, InputRange=(X=-20,Y=-100), TwistAxis="1,0,0")
   ) -> (Output_Get:double)
   n4 = GetOwnerSkeletalMeshComponent()
   n5 = BreakRotator(InRot=BreakTransform(InTransform=GetDeltaTransformFromRefPose(self=@n4, BoneName=thigh_r, BaseName=pelvis)).Rotation)
   > Set thigh_rot_up_r(thigh_rot_up_r:double=MapRangeClamped(Value=@n5.Yaw, InRangeA=-15, InRangeB=-100)) -> (Output_Get:double)
   > Set thigh_rot_inr_r(thigh_rot_inr_r:double=MapRangeClamped(Value=@n5.Pitch, InRangeA=0, InRangeB=-10, OutRangeA=1, OutRangeB=0)) -> (Output_Get:double)
   > Set thigh_rot_otr_r(thigh_rot_otr_r:double=MapRangeClamped(Value=@n5.Pitch, InRangeA=20, InRangeB=50)) -> (Output_Get:double)
   n6 = GetOwnerSkeletalMeshComponent()
   n7 = BreakRotator(InRot=BreakTransform(InTransform=GetDeltaTransformFromRefPose(self=@n6, BoneName=thigh_l, BaseName=pelvis)).Rotation)
   > Set thigh_rot_up_l(thigh_rot_up_l:double=MapRangeClamped(Value=@n7.Yaw, InRangeA=-10, InRangeB=-100)) -> (Output_Get:double)
   > Set thigh_rot_inr_l(thigh_rot_inr_l:double=MapRangeClamped(Value=@n7.Pitch, InRangeA=0, InRangeB=-10, OutRangeA=1, OutRangeB=0)) -> (Output_Get:double)
   > Set thigh_rot_otr_l(thigh_rot_otr_l:double=MapRangeClamped(Value=@n7.Pitch, InRangeA=20, InRangeB=50)) -> (Output_Get:double)
   > Set jumpingAlpha(jumpingAlpha:double="To Float (Boolean)"(InBool=Get bIsJumping())) -> (Output_Get:double)
   n8 = Get bIsMoving2D()
   > Set isNotCrouchRunning(isNotCrouchRunning:double="To Float (Boolean)"(InBool=NOT Boolean(A=@n8))) -> (Output_Get:double)
   > Set ParachuteAlpha(ParachuteAlpha:double="To Float (Boolean)"(InBool=Get bIsParachuteOpened())) -> (Output_Get:double)
   > Set skydiveAlpha(skydiveAlpha:double="To Float (Boolean)"(InBool=Get bIsSkyDiving())) -> (Output_Get:double)
   > Set Not Crouching(Not Crouching:bool=NOT Boolean(A=Get bIsCrouching())) -> (Output_Get:bool)
   n9 = CheckForBackbling()
   Set wearingBackpack(execute=@n1.then_3, wearingBackpack:bool=AND Boolean(A=@n9, B=NOT Boolean(A=@n9.bIsBackblingACape))) -> (Output_Get:bool)
   n10 = Branch(execute=@n1.then_4, Condition=Get bIsCrouching())
   > Set Coat Rigids(Coat Rigids:double=0) -> (Output_Get:double)
   Set Coat Rigids(execute=@n10.else, Coat Rigids=1)
   n11 = Branch(execute=@n1.then_5, Condition=Get bIsTargeting())
   > Set Coat Rigids(Coat Rigids=0.5)
   Set Coat Rigids(execute=@n11.else, Coat Rigids=1)
   n12 = PlayerMovement(execute=@n1.then_6, AnimInstance=Self-Reference(), XDotProductCoefficient=1, YDotProductCoefficient=1, ZDotProductCoefficient=1)
   n13 = MapRangeClamped(Value=@n12.YMovement, InRangeA=0, InRangeB=-1)
   n14 = float * float(A="Absolute (Float)"(A=@n12), B=0.5)
   > Set BwdMovement(BwdMovement:double=MapRangeClamped(
     Value=float + float(A=@n14, B=@n13) {is=title K2Node_CommutativeAssociativeBinaryOperator}
     InRangeA=0
     InRangeB=1.1
     OutRangeA=0.4
     OutRangeB=0
    )) -> (Output_Get:double)
   n15 = MapRangeClamped(Value=@n12, InRangeA=-0.2, InRangeB=1)
   n16 = float * float(A="Absolute (Float)"(A=@n12.YMovement), B=0.2)
   > Set RightMovement(RightMovement:double=MapRangeClamped(Value=float + float(A=@n16, B=@n15), InRangeA=0, InRangeB=1.2, OutRangeB=0.5)) -> (Output_Get:double)
   n17 = MapRangeClamped(Value=@n12, InRangeA=0.2, InRangeB=-1, OutRangeB=0.3)
   n18 = float * float(A="Absolute (Float)"(A=@n12.YMovement), B=0.2)
   > Set LeftMovement(LeftMovement:double=MapRangeClamped(Value=float + float(A=@n18, B=@n17), InRangeA=0, InRangeB=1.2, OutRangeB=0.5)) -> (Output_Get:double)
   n19 = Branch(execute=@n1.then_7, Condition=Get bIsSkyDiving())
   > n20 = Branch(Condition="float > float"(A=Get PawnVelocityZ(), B=50) {is=title K2Node_CallFunction})
   > Set BwdMovement(BwdMovement=float * float(A=Get BwdMovement(), B=0))
   Set BwdMovement(execute=[@n19.else, @n20.else], BwdMovement=float * float(A=Get BwdMovement(), B=1))
   Event BlueprintInitializeAnimation()
   > Set Skel_Mesh(Skel_Mesh=GetOwnerSkeletalMeshComponent())
   n21 = Branch(Condition=Get bIsSwimSprinting() -> (bIsSwimSprinting:bool))
   > Set gravity_override(gravity_override="0, 0, 0")
   n22 = Branch(execute=@n21.else, Condition=Get bIsSkyDiving())
   > n23 = Branch(Condition=Get bIsParachuteOpened())
   Set gravity_override(execute=[@n22.else, @n23.then], gravity_override="0, 0,-980")
   Set gravity_override(execute=@n23.else, gravity_override="0, 0,-380")
   Get bIsCrouchSprinting()
   Set crouching vel(crouching vel:Vector="0, 0, 0") -> (Output_Get:Vector)
   Set crouching acc(crouching acc:Vector="0, 0, 0") -> (Output_Get:Vector)
  ]}
]
variables [
 BwdMovement {type=number}
 Coat Rigids {type=number}
 crouching acc {type=Vector}
 crouching vel {type=Vector}
 Gravity Override Params {
  type=GravityOverrideParamsStruct
  default {
   skydive_FloatingDown {gravityOverride=<0 -700 0>, windAmplitude=<200 -500 100>, windOffset=<0 -600 100>}
   skydive_GoingUp {gravityOverride=<0 -300 200>, windAmplitude=<200 -500 100>}
   skydive_Parachuting {gravityOverride=<0 -500 -980>, windAmplitude=<100 -300 100>}
  }
 }
 gravity_override {type=Vector}
 hand l rot {type=number}
 InComponentSpace {type=boolean}
 InComponentSpaceLastFrame {type=boolean}
 InterpSpeed {type=number}
 isNotCrouchRunning {type=number}
 isNotWearingBackbling {type=number}
 isWearingBackbling {type=boolean}
 isWearingBackpack {type=number}
 isWearingCape {type=number}
 jumpingAlpha {type=number}
 LeftMovement {type=number}
 LockedOnSimSpace {type=boolean}
 Not Crouching {type=boolean}
 ParachuteAlpha {type=number}
 RightMovement {type=number}
 sim_space_rotation {type=Rotator}
 Skel_Mesh {type=SkeletalMeshComponent}
 skydiveAlpha {type=number}
 thigh_rot_inr_l {type=number}
 thigh_rot_inr_r {type=number}
 thigh_rot_otr_l {type=number}
 thigh_rot_otr_r {type=number}
 thigh_rot_up_l {type=number}
 thigh_rot_up_r {type=number}
 wearingBackpack {type=boolean}
 WearingCape {type=boolean}
]
