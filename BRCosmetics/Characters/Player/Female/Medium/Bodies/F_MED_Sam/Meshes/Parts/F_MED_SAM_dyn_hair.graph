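# Physics asset for the dynamic back hair on the female medium Sam character body.
# Capsules on the head, neck, upper spine, clavicles and upper arms form the body shapes around the hair.
# Center, left and right hair chains start at MSTR bodies constrained to spine_05, with limited swing and very tight twist between links.
# Cross constraints tie both side-chain tips to the center chain's third link, mostly with free rotation.
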PhysicsAsset /BRCosmetics/Characters/Player/Female/Medium/Bodies/F_MED_Sam/Meshes/Parts/F_MED_SAM_dyn_hair
bodies {
 clavicle_l [{shapeType=Capsule, length=5.4889, radius=7.5166, center=<5.8792 0.5 -2.7879>, rotation=<rot 84.9998 0 0>}]
 clavicle_r [{shapeType=Capsule, length=5.0266, radius=7.5508, center=<-5.5199 -0.4614 2.7331>, rotation=<rot 90 3.7927 3.7927>}]
 dyn_hair_bk_01 [{shapeType=Capsule, length=6.4316, radius=4.1067, shapeName=dyn_hair_bk_01_capsule, center=<3 0 2.25>, rotation=<rot 0 0 90>}]
 dyn_hair_bk_02 [{shapeType=Capsule, length=6.4316, radius=4.1067, shapeName=dyn_hair_bk_02_capsule, center=<3 0 2.5>, rotation=<rot 0 0 90>}]
 dyn_hair_bk_03 [{shapeType=Capsule, length=6.4316, radius=4.1067, shapeName=dyn_hair_bk_03_capsule, center=<3 0 2.5>, rotation=<rot 0 0 90>}]
 dyn_hair_bk_04 [{shapeType=Capsule, length=6.4316, radius=4.1067, shapeName=dyn_hair_bk_04_capsule, center=<3 0 2.5>, rotation=<rot 0 0 90>}]
 dyn_hair_bk_MSTR [{shapeType=Capsule, length=1, radius=1, rotation=<rot 90 0 0>}]
 dyn_hair_bk_l_01 [{shapeType=Capsule, length=5.4012, radius=1.789, center=<3 0 0>, rotation=<rot 90 0 0>}]
 dyn_hair_bk_l_02 [{shapeType=Capsule, length=3.4658, radius=1.5491, center=<2 0 0>, rotation=<rot 90 0 0>}]
 dyn_hair_bk_l_03 [{shapeType=Capsule, length=3, radius=1.5, center=<2 0 0>, rotation=<rot 90 0 0>}]
 dyn_hair_bk_l_MSTR [{shapeType=Capsule, length=1, radius=1, rotation=<rot 90 0 0>}]
 dyn_hair_bk_r_01 [{shapeType=Capsule, length=4.5882, radius=2.2401, center=<3 0 0>, rotation=<rot 90 0 0>}]
 dyn_hair_bk_r_02 [{shapeType=Capsule, length=3.3751, radius=1.4403, center=<2.4675 -0.0001 0>, rotation=<rot 90 0 0>}]
 dyn_hair_bk_r_03 [{shapeType=Capsule, length=3.25, radius=1.25, center=<2 0 0>, rotation=<rot 90 0 0>}]
 dyn_hair_bk_r_MSTR [{shapeType=Capsule, length=1, radius=1, rotation=<rot 90 0 0>}]
 head [
  {shapeType=Capsule, length=6.1617, radius=5.4789, center=<1.7962 1.5698 -0.0142>, rotation=<rot 90 0 0>}
  {shapeType=Capsule, length=6.6761, radius=4.9645, center=<1.7962 3.9626 0.1708>, rotation=<rot 90 0 0>}
 ]
 neck_01 [{shapeType=Capsule, length=10.6829, radius=5.4577, center=<-0.9235 -0.2989 0>, rotation=<rot 90 0 -10>}]
 spine_02 [{shapeType=Capsule, length=6.6622, radius=10.4318, center=<0 -1.6894 0.0002>, rotation=<rot 0 -90 0>}]
 spine_03 [{shapeType=Capsule, length=6.2951, radius=10.7988, center=<0 -0.1996 0>, rotation=<rot 0 -90 0>}]
 spine_04 [{shapeType=Capsule, length=4.7542, radius=11.2793, center=<2.75 1.4464 0.0001>, rotation=<rot 0 -90 0>}]
 spine_05 [{shapeType=Capsule, length=13.1239, radius=8.741, center=<4.2135 1.7196 0>, rotation=<rot 0 -90 0>}]
 upperarm_l [{shapeType=Capsule, length=14.3652, radius=6.5693, center=<10.0001 0.5147 -2>, rotation=<rot 90 0 0>}]
 upperarm_r [{shapeType=Capsule, length=11.9345, radius=7, center=<-10 0 2>, rotation=<rot 90 0 0>}]
}
constraints [
 {
  bone1Name=head
  bone2Name=neck_01
  swing1LimitDegrees=45
  swing1Motion=Free
  swing2LimitDegrees=45
  swing2Motion=Free
  twistLimitDegrees=45
  twistMotion=Free
 }
 {
  bone1Name=spine_05
  bone2Name=spine_04
  swing1LimitDegrees=45
  swing1Motion=Free
  swing2LimitDegrees=45
  swing2Motion=Free
  twistLimitDegrees=45
  twistMotion=Free
 }
 {
  bone1Name=spine_04
  bone2Name=spine_03
  swing1LimitDegrees=45
  swing1Motion=Free
  swing2LimitDegrees=45
  swing2Motion=Free
  twistLimitDegrees=45
  twistMotion=Free
 }
 {
  bone1Name=spine_03
  bone2Name=spine_02
  swing1LimitDegrees=45
  swing1Motion=Free
  swing2LimitDegrees=45
  swing2Motion=Free
  twistLimitDegrees=45
  twistMotion=Free
 }
 {
  bone1Name=neck_01
  bone2Name=spine_05
  swing1LimitDegrees=45
  swing1Motion=Free
  swing2LimitDegrees=45
  swing2Motion=Free
  twistLimitDegrees=45
  twistMotion=Free
 }
 {
  bone1Name=clavicle_l
  bone2Name=spine_05
  swing1LimitDegrees=45
  swing1Motion=Free
  swing2LimitDegrees=45
  swing2Motion=Free
  twistLimitDegrees=45
  twistMotion=Free
 }
 {
  bone1Name=clavicle_r
  bone2Name=spine_05
  swing1LimitDegrees=45
  swing1Motion=Free
  swing2LimitDegrees=45
  swing2Motion=Free
  twistLimitDegrees=45
  twistMotion=Free
 }
 {
  bone1Name=upperarm_r
  bone2Name=clavicle_r
  swing1LimitDegrees=45
  swing1Motion=Free
  swing2LimitDegrees=45
  swing2Motion=Free
  twistLimitDegrees=45
  twistMotion=Free
 }
 {
  bone1Name=upperarm_l
  bone2Name=clavicle_l
  swing1LimitDegrees=45
  swing1Motion=Free
  swing2LimitDegrees=45
  swing2Motion=Free
  twistLimitDegrees=45
  twistMotion=Free
 }
 {
  bone1Name=dyn_hair_bk_MSTR
  bone2Name=spine_05
  swing1LimitDegrees=45
  swing1Motion=Free
  swing2LimitDegrees=45
  swing2Motion=Free
  twistLimitDegrees=45
  twistMotion=Free
 }
 {bone1Name=dyn_hair_bk_01, bone2Name=dyn_hair_bk_MSTR, swing1LimitDegrees=20, swing2LimitDegrees=30, twistLimitDegrees=2}
 {bone1Name=dyn_hair_bk_02, bone2Name=dyn_hair_bk_01, swing1LimitDegrees=20, swing2LimitDegrees=30, twistLimitDegrees=2}
 {bone1Name=dyn_hair_bk_03, bone2Name=dyn_hair_bk_02, swing1LimitDegrees=20, swing2LimitDegrees=30, twistLimitDegrees=2}
 {bone1Name=dyn_hair_bk_04, bone2Name=dyn_hair_bk_03, swing1LimitDegrees=20, swing2LimitDegrees=30, twistLimitDegrees=2}
 {
  bone1Name=dyn_hair_bk_l_MSTR
  bone2Name=spine_05
  swing1LimitDegrees=45
  swing1Motion=Free
  swing2LimitDegrees=45
  swing2Motion=Free
  twistLimitDegrees=45
  twistMotion=Free
 }
 {bone1Name=dyn_hair_bk_l_01, bone2Name=dyn_hair_bk_l_MSTR, swing1LimitDegrees=20, swing2LimitDegrees=30, twistLimitDegrees=2}
 {bone1Name=dyn_hair_bk_l_02, bone2Name=dyn_hair_bk_l_01, swing1LimitDegrees=10, swing2LimitDegrees=30, twistLimitDegrees=2}
 {bone1Name=dyn_hair_bk_l_03, bone2Name=dyn_hair_bk_l_02, swing1LimitDegrees=10, swing2LimitDegrees=30, twistLimitDegrees=2}
 {
  bone1Name=dyn_hair_bk_r_MSTR
  bone2Name=spine_05
  swing1LimitDegrees=45
  swing1Motion=Free
  swing2LimitDegrees=45
  swing2Motion=Free
  twistLimitDegrees=45
  twistMotion=Free
 }
 {bone1Name=dyn_hair_bk_r_01, bone2Name=dyn_hair_bk_r_MSTR, swing1LimitDegrees=20, swing2LimitDegrees=30, twistLimitDegrees=2}
 {bone1Name=dyn_hair_bk_r_02, bone2Name=dyn_hair_bk_r_01, swing1LimitDegrees=20, swing2LimitDegrees=30, twistLimitDegrees=2}
 {bone1Name=dyn_hair_bk_r_03, bone2Name=dyn_hair_bk_r_02, swing1LimitDegrees=20, swing2LimitDegrees=30, twistLimitDegrees=2}
 {
  bone1Name=dyn_hair_bk_r_03
  bone2Name=dyn_hair_bk_03
  swing1LimitDegrees=45
  swing1Motion=Free
  swing2LimitDegrees=45
  swing2Motion=Free
  twistLimitDegrees=45
  twistMotion=Free
 }
 {
  bone1Name=dyn_hair_bk_l_03
  bone2Name=dyn_hair_bk_03
  swing1LimitDegrees=5
  swing2LimitDegrees=45
  swing2Motion=Free
  twistLimitDegrees=45
  twistMotion=Free
 }
]
