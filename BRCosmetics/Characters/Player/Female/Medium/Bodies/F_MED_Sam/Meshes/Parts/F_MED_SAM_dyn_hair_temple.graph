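# Physics asset for the temple hair of the F_MED_Sam player cosmetic, defining bodies and joints for left and right dynamic hair chains.
# Capsules on the head, neck, upper spine, clavicles and upper-arm twist bones form the body volumes around the hair, joined by free-angular constraints.
# Hair segments chain through swing-limited, tight-twist joints from free-rotating MSTR roots.
# The left root attaches to head, while the right attaches to neck_01 and continues into an "in" sub-chain.
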PhysicsAsset /BRCosmetics/Characters/Player/Female/Medium/Bodies/F_MED_Sam/Meshes/Parts/F_MED_SAM_dyn_hair_temple
bodies {
 clavicle_l [{shapeType=Capsule, length=4.8681, radius=6.583, center=<6.6159 -0.7345 -2.1362>, rotation=<rot 84.9998 0 0>}]
 clavicle_r [{shapeType=Capsule, length=6.7971, radius=6.4041, center=<-5.6834 0.9414 1.6838>, rotation=<rot 85 0 0>}]
 dyn_hair_l_01 [{shapeType=Capsule, length=2, radius=1, center=<3 0 0>, rotation=<rot 90 0 0>}]
 dyn_hair_l_02 [{shapeType=Capsule, length=4, radius=1.5, center=<3.5 0 0>, rotation=<rot 90 0 0>}]
 dyn_hair_l_03 [{shapeType=Capsule, length=3, radius=1.5, center=<3 0 0>, rotation=<rot 90 0 0>}]
 dyn_hair_l_04 [{shapeType=Capsule, length=3, radius=1.5, center=<3 0 0>, rotation=<rot 90 0 0>}]
 dyn_hair_l_05 [{shapeType=Capsule, length=3, radius=1.5, center=<3 0 0>, rotation=<rot 90 0 0>}]
 dyn_hair_l_MSTR [{shapeType=Capsule, length=1, radius=1, rotation=<rot 90 0 0>}]
 dyn_hair_r_01 [{shapeType=Capsule, length=2.3337, radius=2.127, shapeName=dyn_hair_r_01_capsule, center=<2 0 0>, rotation=<rot 90 0 0>}]
 dyn_hair_r_02 [{shapeType=Capsule, length=2.97, radius=2.805, shapeName=dyn_hair_r_02_capsule, center=<2 0 0>, rotation=<rot 90 0 0>}]
 dyn_hair_r_MSTR [{shapeType=Capsule, length=1, radius=1, rotation=<rot 90 0 0>}]
 dyn_hair_r_in_01 [{shapeType=Capsule, length=3, radius=1.5, center=<3 0 0>, rotation=<rot 90 0 0>}]
 dyn_hair_r_in_02 [{shapeType=Capsule, length=4, radius=1.5, center=<3.5 0 0>, rotation=<rot 90 0 0>}]
 dyn_hair_r_in_03 [{shapeType=Capsule, length=3, radius=1.5, center=<3 0 0>, rotation=<rot 90 0 0>}]
 dyn_hair_r_in_04 [{shapeType=Capsule, length=3, radius=1.5, center=<3 0 0>, rotation=<rot 90 0 0>}]
 head [
  {shapeType=Capsule, length=6.1617, radius=5.4789, center=<1.7962 1.5698 -0.0142>, rotation=<rot 90 0 0>}
  {shapeType=Capsule, length=6.6761, radius=4.9645, center=<1.7962 3.9626 0.1708>, rotation=<rot 90 0 0>}
 ]
 neck_01 [{shapeType=Capsule, length=6.462, radius=5.1786, center=<4.1718 0 0>, rotation=<rot 90 0 0>}]
 spine_04 [{shapeType=Capsule, length=11.0042, radius=11.5293, center=<0.108 0.4915 0.0001>, rotation=<rot 0 -90 0>}]
 spine_05 [
  {shapeType=Capsule, length=9.9499, radius=9.1841, center=<3.591 -1.9667 0>, rotation=<rot 0 -90 0>}
  {shapeType=Capsule, length=13.6999, radius=9.1841, center=<-1.7059 0.596 0>, rotation=<rot 0 -90 0>}
 ]
 upperarm_twist_02_l [{shapeType=Capsule, length=13.7011, radius=5.2334, center=<-5 0 -2>, rotation=<rot 90 0 0>}]
 upperarm_twist_02_r [{shapeType=Capsule, length=12.9511, radius=5.9834, center=<7.817 0.2464 1.4597>, rotation=<rot 90 -0.5968 -5.5968>}]
}
constraints [
 {
  bone1Name=head
  bone2Name=neck_01
  swing1LimitDegrees=45
  swing1Motion=Free
  swing2LimitDegrees=45
  swing2Motion=Free
  twistLimitDegrees=45
  twistMotion=Free
 }
 {
  bone1Name=spine_05
  bone2Name=spine_04
  swing1LimitDegrees=45
  swing1Motion=Free
  swing2LimitDegrees=45
  swing2Motion=Free
  twistLimitDegrees=45
  twistMotion=Free
 }
 {
  bone1Name=neck_01
  bone2Name=spine_05
  swing1LimitDegrees=45
  swing1Motion=Free
  swing2LimitDegrees=45
  swing2Motion=Free
  twistLimitDegrees=45
  twistMotion=Free
 }
 {
  bone1Name=clavicle_l
  bone2Name=spine_05
  swing1LimitDegrees=45
  swing1Motion=Free
  swing2LimitDegrees=45
  swing2Motion=Free
  twistLimitDegrees=45
  twistMotion=Free
 }
 {
  bone1Name=clavicle_r
  bone2Name=spine_05
  swing1LimitDegrees=45
  swing1Motion=Free
  swing2LimitDegrees=45
  swing2Motion=Free
  twistLimitDegrees=45
  twistMotion=Free
 }
 {bone1Name=dyn_hair_l_01, bone2Name=dyn_hair_l_MSTR, swing1LimitDegrees=30, swing2LimitDegrees=30, twistLimitDegrees=2}
 {bone1Name=dyn_hair_l_02, bone2Name=dyn_hair_l_01, swing1LimitDegrees=30, swing2LimitDegrees=30, twistLimitDegrees=2}
 {bone1Name=dyn_hair_l_03, bone2Name=dyn_hair_l_02, swing1LimitDegrees=30, swing2LimitDegrees=30, twistLimitDegrees=2}
 {bone1Name=dyn_hair_l_04, bone2Name=dyn_hair_l_03, swing1LimitDegrees=30, swing2LimitDegrees=30, twistLimitDegrees=2}
 {bone1Name=dyn_hair_l_05, bone2Name=dyn_hair_l_04, swing1LimitDegrees=30, swing2LimitDegrees=30, twistLimitDegrees=2}
 {
  bone1Name=dyn_hair_l_MSTR
  bone2Name=head
  swing1LimitDegrees=45
  swing1Motion=Free
  swing2LimitDegrees=45
  swing2Motion=Free
  twistLimitDegrees=45
  twistMotion=Free
 }
 {
  bone1Name=upperarm_twist_02_l
  bone2Name=clavicle_l
  swing1LimitDegrees=45
  swing1Motion=Free
  swing2LimitDegrees=45
  swing2Motion=Free
  twistLimitDegrees=45
  twistMotion=Free
 }
 {
  bone1Name=upperarm_twist_02_r
  bone2Name=clavicle_r
  swing1LimitDegrees=45
  swing1Motion=Free
  swing2LimitDegrees=45
  swing2Motion=Free
  twistLimitDegrees=45
  twistMotion=Free
 }
 {
  bone1Name=dyn_hair_r_MSTR
  bone2Name=neck_01
  swing1LimitDegrees=45
  swing1Motion=Free
  swing2LimitDegrees=45
  swing2Motion=Free
  twistLimitDegrees=45
  twistMotion=Free
 }
 {bone1Name=dyn_hair_r_01, bone2Name=dyn_hair_r_MSTR, swing1LimitDegrees=30, swing2LimitDegrees=30, twistLimitDegrees=2}
 {bone1Name=dyn_hair_r_02, bone2Name=dyn_hair_r_01, swing1LimitDegrees=30, swing2LimitDegrees=30, twistLimitDegrees=2}
 {bone1Name=dyn_hair_r_in_01, bone2Name=dyn_hair_r_02, swing1LimitDegrees=30, swing2LimitDegrees=30, twistLimitDegrees=2}
 {bone1Name=dyn_hair_r_in_02, bone2Name=dyn_hair_r_in_01, swing1LimitDegrees=30, swing2LimitDegrees=30, twistLimitDegrees=2}
 {bone1Name=dyn_hair_r_in_03, bone2Name=dyn_hair_r_in_02, swing1LimitDegrees=30, swing2LimitDegrees=30, twistLimitDegrees=2}
 {bone1Name=dyn_hair_r_in_04, bone2Name=dyn_hair_r_in_03, swing1LimitDegrees=30, swing2LimitDegrees=30, twistLimitDegrees=2}
]
