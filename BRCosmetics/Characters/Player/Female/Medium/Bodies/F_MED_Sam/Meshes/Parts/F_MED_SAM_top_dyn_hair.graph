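# Physics asset for the dynamic back hair on the F_MED_Sam player character's top mesh part.
# Three capsule bodies cover the hair base bone, the neck and the upper spine, giving the hair simple body shapes around it.
# The hair base is constrained to the neck with tight twist and asymmetric swing limits, while the neck-to-spine joint leaves all angular axes free.
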PhysicsAsset /BRCosmetics/Characters/Player/Female/Medium/Bodies/F_MED_Sam/Meshes/Parts/F_MED_SAM_top_dyn_hair
bodies {
 dyn_hair_bk_base [{shapeType=Capsule, length=10, radius=2, center=<15 0 0>, rotation=<rot 0 0 90>}]
 neck_01 [{shapeType=Capsule, length=10.6829, radius=5.4577, center=<-0.9235 -0.2989 0>, rotation=<rot 90 0 -10>}]
 spine_05 [{shapeType=Capsule, length=13.1239, radius=8.741, center=<4.2135 1.7196 0>, rotation=<rot 0 -90 0>}]
}
constraints [
 {
  bone1Name=neck_01
  bone2Name=spine_05
  swing1LimitDegrees=45
  swing1Motion=Free
  swing2LimitDegrees=45
  swing2Motion=Free
  twistLimitDegrees=45
  twistMotion=Free
 }
 {bone1Name=dyn_hair_bk_base, bone2Name=neck_01, swing1LimitDegrees=10, swing2LimitDegrees=35, twistLimitDegrees=2}
]
