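# Post-process style hair animation Blueprint for the F_MED_Sam player body, adding secondary motion to the dyn_hair bones.
# It copies the parent mesh pose, then applies Look At, RigidBody temple simulation (including a frontend variant) and AnimDynamics.
# From cached pose 'prebackpack', wearingBackpack selects a reduced AnimDynamics chain or the full gravity-override chain with skydive rotation offsets.
# Both chains use partial-alpha constraints to spine and clavicle bones.
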AnimBlueprint /BRCosmetics/Characters/Player/Female/Medium/Bodies/F_MED_Sam/Meshes/Parts/F_MED_SAM_AnimBP
target_skeleton=&/F_MED_SAM_Skeleton
graphs [
 AnimGraph {nodes [
   n1 = "Constraint - dyn_hair_bk_l_04 (R)"(Alpha=1, ConstraintWeights_0=1) {
    alphaBoolBlend {blendOption=Linear}
    boneToModify {boneName=dyn_hair_bk_l_04}
    constraintSetup [{transformType=Rotation, targetBone {boneName=dyn_hair_bk_04}}]
    constraintWeights [1]
   }
   n2 = "Constraint - dyn_hair_bk_r_04 (R)"(ComponentPose=@n1, Alpha=1, ConstraintWeights_0=1) {
    alphaBoolBlend {blendOption=Linear}
    boneToModify {boneName=dyn_hair_bk_r_04}
    constraintSetup [{transformType=Rotation, targetBone {boneName=dyn_hair_bk_04}}]
    constraintWeights [1]
   }
   n3 = Anim Dynamics - Bone: dyn_hair_bk_l_04(ComponentPose=@n2, Alpha=1, ExternalForce="0,0,0") {
    angularDampingOverride=0.9999
    angularSpringConstant=100
    bGravityOverrideInSimSpace=true
    bOverrideAngularDamping=true
    simulationSpace=RootRelative
    alphaBoolBlend {blendOption=Linear}
    boundBone {boneName=dyn_hair_bk_l_04}
    componentAppliedLinearAccClamp=<3500 3500 3500>
    componentLinearAccScale=<0.6 0.6 0.8>
    componentLinearVelScale=<0.6 0.6 0.8>
    physicsBodyDefinitions [{
      sphereCollisionRadius=0
      boundBone {boneName=dyn_hair_bk_l_04}
      boxExtents=<20 20 10>
      constraintSetup {angularLimitsMax=<1 5 10>, angularLimitsMin=<-1 -30 -10>, angularTarget=<0 0 0>}
      localJointOffset=<1 0 0>
     }]
    relativeSpaceBone {boneName=root}
   }
   n4 = Anim Dynamics - Bone: dyn_hair_bk_l_05(ComponentPose=@n3, Alpha=1, ExternalForce="0,0,0") {
    angularDampingOverride=0.9999
    angularSpringConstant=100
    bGravityOverrideInSimSpace=true
    bOverrideAngularDamping=true
    simulationSpace=RootRelative
    alphaBoolBlend {blendOption=Linear}
    boundBone {boneName=dyn_hair_bk_l_05}
    componentAppliedLinearAccClamp=<3500 3500 3500>
    componentLinearAccScale=<0.6 0.6 0.8>
    componentLinearVelScale=<0.6 0.6 0.8>
    physicsBodyDefinitions [{
      sphereCollisionRadius=0
      boundBone {boneName=dyn_hair_bk_l_05}
      boxExtents=<20 20 10>
      constraintSetup {angularLimitsMax=<1 5 10>, angularLimitsMin=<-1 -30 -10>, angularTarget=<0 0 0>}
      localJointOffset=<1 0 0>
     }]
    relativeSpaceBone {boneName=root}
   }
   Anim Dynamics - Bone: dyn_hair_bk_r_04(ComponentPose=@n4, Alpha=1, ExternalForce="0,0,0") {
    angularDampingOverride=0.9999
    angularSpringConstant=100
    bGravityOverrideInSimSpace=true
    bOverrideAngularDamping=true
    simulationSpace=RootRelative
    alphaBoolBlend {blendOption=Linear}
    boundBone {boneName=dyn_hair_bk_r_04}
    componentAppliedLinearAccClamp=<3500 3500 3500>
    componentLinearAccScale=<0.6 0.6 0.8>
    componentLinearVelScale=<0.6 0.6 0.8>
    physicsBodyDefinitions [{
      sphereCollisionRadius=0
      boundBone {boneName=dyn_hair_bk_r_04}
      boxExtents=<20 20 10>
      constraintSetup {angularLimitsMax=<1 5 5>, angularLimitsMin=<-1 -30 -5>, angularTarget=<0 0 0>}
      localJointOffset=<1 0 0>
     }]
    relativeSpaceBone {boneName=root}
   }
   Anim Dynamics - Bone: dyn_hair_bk_05(Alpha=1, ExternalForce="0,0,0") {
    angularDampingOverride=0.9999
    angularSpringConstant=100
    bGravityOverrideInSimSpace=true
    bOverrideAngularDamping=true
    simulationSpace=RootRelative
    alphaBoolBlend {blendOption=Linear}
    boundBone {boneName=dyn_hair_bk_05}
    componentAppliedLinearAccClamp=<3500 3500 3500>
    componentLinearAccScale=<0.6 0.6 0.8>
    componentLinearVelScale=<0.6 0.6 0.8>
    physicsBodyDefinitions [{
      sphereCollisionRadius=0
      boundBone {boneName=dyn_hair_bk_05}
      boxExtents=<20 20 10>
      constraintSetup {angularLimitsMax=<1 10 10>, angularLimitsMin=<-1 -30 -10>, angularTarget=<0 0 0>}
      localJointOffset=<1 0 0>
     }]
    relativeSpaceBone {boneName=root}
   }
   n5 = Get gravity_override() -> (gravity_override:Vector)
   Anim Dynamics - Bone: dyn_hair_bk_r_01!(ComponentPose, Alpha=1, GravityScale:float=1, GravityOverride:Vector=@n5, ExternalForce="0,0,0") -> (Pose) {
    angularDampingOverride=1
    bAngularSpring=true
    bGravityOverrideInSimSpace=true
    bOverrideAngularDamping=true
    simulationSpace=RootRelative
    alphaBoolBlend {blendOption=Linear}
    boundBone {boneName=dyn_hair_bk_r_01}
    componentAppliedLinearAccClamp=<5000 5000 4000>
    componentLinearAccScale=<1.2 1.2 1.5>
    componentLinearVelScale=<1.2 1.2 1.5>
    gravityOverride {z=-980}
    physicsBodyDefinitions [{
      sphereCollisionRadius=0
      boundBone {boneName=dyn_hair_bk_r_01}
      boxExtents=<20 20 10>
      constraintSetup {angularLimitsMax=<45 5 45>, angularLimitsMin=<-45 -45 -45>, angularTarget=<0 0 0>}
      localJointOffset=<2 0 -0.5>
     }]
    relativeSpaceBone {boneName=root}
   }
   Anim Dynamics - Bone: dyn_hair_bk_l_01!(ComponentPose, Alpha=1, GravityScale:float=1, GravityOverride:Vector=@n5, ExternalForce="0,0,0") -> (Pose) {
    angularDampingOverride=1
    angularSpringConstant=75
    bAngularSpring=true
    bGravityOverrideInSimSpace=true
    bOverrideAngularDamping=true
    simulationSpace=RootRelative
    alphaBoolBlend {blendOption=Linear}
    boundBone {boneName=dyn_hair_bk_l_01}
    componentAppliedLinearAccClamp=<10000 10000 8000>
    componentLinearAccScale=<1.4 1.4 1.6>
    componentLinearVelScale=<1.4 1.4 1.6>
    gravityOverride {z=-980}
    physicsBodyDefinitions [{
      sphereCollisionRadius=0
      boundBone {boneName=dyn_hair_bk_l_01}
      boxExtents=<20 20 10>
      constraintSetup {angularLimitsMax=<45 5 45>, angularLimitsMin=<-45 -45 -45>, angularTarget=<0 0 0>}
      localJointOffset=<2 0 -0.5>
     }]
    relativeSpaceBone {boneName=root}
   }
   "Constraint - dyn_hair_bk_r_01 (P)"(Alpha=0.5, ConstraintWeights_0=1) {
    alphaBoolBlend {blendOption=Linear}
    boneToModify {boneName=dyn_hair_bk_r_01}
    constraintSetup [{targetBone {boneName=spine_05}}]
    constraintWeights [1]
   }
   "Constraint - dyn_hair_bk_r_02 (P)"(Alpha=0.45, ConstraintWeights_0=1) {
    alphaBoolBlend {blendOption=Linear}
    boneToModify {boneName=dyn_hair_bk_r_02}
    constraintSetup [{targetBone {boneName=clavicle_l}}]
    constraintWeights [1]
   }
   "Constraint - dyn_hair_bk_r_01 (P)"(Alpha=0.5, ConstraintWeights_0=1) {
    alphaBoolBlend {blendOption=Linear}
    boneToModify {boneName=dyn_hair_bk_r_01}
    constraintSetup [{targetBone {boneName=spine_05}}]
    constraintWeights [1]
   }
   "Constraint - dyn_hair_bk_r_02 (P)"(Alpha=0.45, ConstraintWeights_0=1) {
    alphaBoolBlend {blendOption=Linear}
    boneToModify {boneName=dyn_hair_bk_r_02}
    constraintSetup [{targetBone {boneName=clavicle_l}}]
    constraintWeights [1]
   }
   RigidBody!(ComponentPose, Alpha, OverrideWorldGravity:Vector="0,0,-980", ExternalForce) -> (Pose) {
    bClampLinearTranslationLimitToRefPose=true
    bOverrideWorldGravity=true
    overridePhysicsAsset=&/F_MED_SAM_top_dyn_hair
    simulationSpace=BaseBoneSpace
    alphaBoolBlend {blendOption=Linear}
    baseBoneRef {boneName=neck_02}
    componentAppliedLinearAccClamp=<1000 1000 1000>
    componentLinearAccScale=<0.2 0.2 0.2>
    componentLinearVelScale=<0.2 0.2 0.2>
    overrideWorldGravity {z=-980}
   }
   RigidBody(OverrideWorldGravity=Get gravity_override()) {
    bClampLinearTranslationLimitToRefPose=true
    bOverrideWorldGravity=true
    overridePhysicsAsset=&/F_MED_SAM_dyn_hair
    simulationSpace=BaseBoneSpace
    alphaBoolBlend {blendOption=Linear}
    baseBoneRef {boneName=spine_05}
    componentAppliedLinearAccClamp=<3000 3000 3000>
    componentLinearAccScale=<0.3 0.3 0.5>
    componentLinearVelScale=<0.3 0.3 0.5>
    overrideWorldGravity {z=-980}
   }
   n6 = Anim Dynamics - Bone: dyn_hair_bk_01!(
    ComponentPose=Local To Component(LocalPose=Use cached pose 'prebackpack'())
    Alpha=1
    GravityScale:float=1
    GravityOverride:Vector=@n5
    ExternalForce="0,0,0"
   ) -> (Pose) {
    angularDampingOverride=1
    angularSpringConstant=50
    bAngularSpring=true
    bGravityOverrideInSimSpace=true
    bOverrideAngularDamping=true
    simulationSpace=RootRelative
    alphaBoolBlend {blendOption=Linear}
    boundBone {boneName=dyn_hair_bk_01}
    componentAppliedLinearAccClamp=<5000 5000 4000>
    componentLinearAccScale=<1.2 1.2 1.3>
    componentLinearVelScale=<1.2 1.2 1.3>
    gravityOverride {z=-980}
    physicsBodyDefinitions [{
      sphereCollisionRadius=0
      boundBone {boneName=dyn_hair_bk_01}
      boxExtents=<12 11 11>
      constraintSetup {angularLimitsMax=<45 5 35>, angularLimitsMin=<-45 -45 -35>, angularTarget=<0 0 0>}
      localJointOffset=<1 0 0>
     }]
    relativeSpaceBone {boneName=root}
   }
   n7 = Anim Dynamics - Bone: dyn_hair_bk_02!(ComponentPose=@n6, Alpha=1, GravityScale:float=1, GravityOverride:Vector=@n5, ExternalForce="0,0,0") -> (Pose) {
    angularDampingOverride=1
    angularSpringConstant=50
    bAngularSpring=true
    bGravityOverrideInSimSpace=true
    bOverrideAngularDamping=true
    simulationSpace=RootRelative
    alphaBoolBlend {blendOption=Linear}
    boundBone {boneName=dyn_hair_bk_02}
    componentAppliedLinearAccClamp=<5000 5000 4000>
    componentLinearAccScale=<1.2 1.2 1.4>
    componentLinearVelScale=<1.2 1.2 1.4>
    gravityOverride {z=-980}
    physicsBodyDefinitions [{
      sphereCollisionRadius=0
      boundBone {boneName=dyn_hair_bk_02}
      boxExtents=<12 11 11>
      constraintSetup {angularLimitsMax=<45 5 35>, angularLimitsMin=<-45 -45 -35>, angularTarget=<0 0 0>}
      localJointOffset=<1 0 0>
     }]
    relativeSpaceBone {boneName=root}
   }
   n8 = Anim Dynamics - Bone: dyn_hair_bk_03!(ComponentPose=@n7, Alpha=1, GravityScale:float=1, GravityOverride:Vector=@n5, ExternalForce="0,0,0") -> (Pose) {
    angularDampingOverride=1
    angularSpringConstant=75
    bAngularSpring=true
    bGravityOverrideInSimSpace=true
    bOverrideAngularDamping=true
    simulationSpace=RootRelative
    alphaBoolBlend {blendOption=Linear}
    boundBone {boneName=dyn_hair_bk_03}
    componentAppliedLinearAccClamp=<5000 5000 5000>
    componentLinearAccScale=<1.2 1.2 1.5>
    componentLinearVelScale=<1.2 1.2 1.5>
    gravityOverride {z=-980}
    physicsBodyDefinitions [{
      sphereCollisionRadius=0
      boundBone {boneName=dyn_hair_bk_03}
      boxExtents=<12 11 11>
      constraintSetup {angularLimitsMax=<45 -5 35>, angularLimitsMin=<-45 -45 -35>, angularTarget=<0 0 0>}
      localJointOffset=<2 0 0>
     }]
    relativeSpaceBone {boneName=root}
   }
   n9 = Anim Dynamics - Bone: dyn_hair_bk_04!(ComponentPose=@n8, Alpha=1, GravityScale:float=1, GravityOverride:Vector=@n5, ExternalForce="0,0,0") -> (Pose) {
    angularDampingOverride=1
    angularSpringConstant=75
    bAngularSpring=true
    bGravityOverrideInSimSpace=true
    bOverrideAngularDamping=true
    simulationSpace=RootRelative
    alphaBoolBlend {blendOption=Linear}
    boundBone {boneName=dyn_hair_bk_04}
    componentAppliedLinearAccClamp=<5000 5000 5000>
    componentLinearAccScale=<1.1 1.1 1.3>
    componentLinearVelScale=<1.1 1.1 1.3>
    gravityOverride {z=-980}
    physicsBodyDefinitions [{
      sphereCollisionRadius=0
      boundBone {boneName=dyn_hair_bk_04}
      boxExtents=<12 11 11>
      constraintSetup {angularLimitsMax=<45 -5 35>, angularLimitsMin=<-45 -45 -35>, angularTarget=<0 0 0>}
      localJointOffset=<2 0 0>
     }]
    relativeSpaceBone {boneName=root}
   }
   n10 = Anim Dynamics - Bone: dyn_hair_bk_05!(ComponentPose=@n9, Alpha=1, GravityScale:float=1, GravityOverride:Vector=@n5, ExternalForce="0,0,0") -> (Pose) {
    angularDampingOverride=1
    angularSpringConstant=100
    bAngularSpring=true
    bGravityOverrideInSimSpace=true
    bOverrideAngularDamping=true
    simulationSpace=RootRelative
    alphaBoolBlend {blendOption=Linear}
    boundBone {boneName=dyn_hair_bk_05}
    componentAppliedLinearAccClamp=<5000 5000 5000>
    componentLinearAccScale=<1 1 1.2>
    componentLinearVelScale=<1 1 1.2>
    gravityOverride {z=-980}
    physicsBodyDefinitions [{
      sphereCollisionRadius=0
      boundBone {boneName=dyn_hair_bk_05}
      boxExtents=<12 12 11>
      constraintSetup {angularLimitsMax=<45 5 35>, angularLimitsMin=<-45 -45 -35>, angularTarget=<0 0 0>}
      localJointOffset=<2 0 0>
     }]
    relativeSpaceBone {boneName=root}
   }
   n11 = Anim Dynamics - Bone: dyn_hair_bk_06!(ComponentPose=@n10, Alpha=1, GravityOverride:Vector=@n5, ExternalForce="0,0,0") -> (Pose) {
    angularDampingOverride=0.9999
    angularSpringConstant=100
    bAngularSpring=true
    bGravityOverrideInSimSpace=true
    bOverrideAngularDamping=true
    bUseGravityOverride=true
    simulationSpace=RootRelative
    alphaBoolBlend {blendOption=Linear}
    boundBone {boneName=dyn_hair_bk_06}
    componentAppliedLinearAccClamp=<5000 5000 5000>
    componentLinearAccScale=<0.9 0.9 0.9>
    componentLinearVelScale=<0.9 0.9 0.9>
    physicsBodyDefinitions [{
      sphereCollisionRadius=0
      boundBone {boneName=dyn_hair_bk_06}
      boxExtents=<13 13 12>
      constraintSetup {angularLimitsMax=<1 5 10>, angularLimitsMin=<-1 -30 -10>, angularTarget=<0 0 0>}
      localJointOffset=<2 0 0>
     }]
    relativeSpaceBone {boneName=root}
   }
   n12 = Anim Dynamics - Bone: dyn_hair_bk_07!(ComponentPose=@n11, Alpha=1, GravityOverride:Vector=@n5, ExternalForce="0,0,0") -> (Pose) {
    angularDampingOverride=0.9999
    angularSpringConstant=100
    bAngularSpring=true
    bGravityOverrideInSimSpace=true
    bOverrideAngularDamping=true
    bUseGravityOverride=true
    simulationSpace=RootRelative
    alphaBoolBlend {blendOption=Linear}
    boundBone {boneName=dyn_hair_bk_07}
    componentAppliedLinearAccClamp=<5000 5000 5000>
    componentLinearAccScale=<0.8 0.8 0.8>
    componentLinearVelScale=<0.8 0.8 0.8>
    physicsBodyDefinitions [{
      sphereCollisionRadius=0
      boundBone {boneName=dyn_hair_bk_07}
      boxExtents=<13 13 12>
      constraintSetup {angularLimitsMax=<1 5 10>, angularLimitsMin=<-1 -30 -10>, angularTarget=<0 0 0>}
      localJointOffset=<2 0 0>
     }]
    relativeSpaceBone {boneName=root}
   }
   n13 = Anim Dynamics - Bone: dyn_hair_bk_l_02!(ComponentPose=@n12, Alpha=1, GravityScale:float=1, GravityOverride:Vector=@n5, ExternalForce="0,0,0") -> (Pose) {
    angularDampingOverride=1
    angularSpringConstant=75
    bAngularSpring=true
    bGravityOverrideInSimSpace=true
    bOverrideAngularDamping=true
    simulationSpace=RootRelative
    alphaBoolBlend {blendOption=Linear}
    boundBone {boneName=dyn_hair_bk_l_02}
    componentAppliedLinearAccClamp=<10000 10000 8000>
    componentLinearAccScale=<1.4 1.4 1.6>
    componentLinearVelScale=<1.4 1.4 1.6>
    gravityOverride {z=-980}
    physicsBodyDefinitions [{
      sphereCollisionRadius=0
      boundBone {boneName=dyn_hair_bk_l_02}
      boxExtents=<20 20 10>
      constraintSetup {angularLimitsMax=<45 0 45>, angularLimitsMin=<-45 -45 -45>, angularTarget=<0 0 0>}
      localJointOffset=<2 0 -0.5>
     }]
    relativeSpaceBone {boneName=root}
   }
   n14 = Anim Dynamics - Bone: dyn_hair_bk_l_03!(ComponentPose=@n13, Alpha=1, GravityScale:float=1, GravityOverride:Vector=@n5, ExternalForce="0,0,0") -> (Pose) {
    angularDampingOverride=1
    angularSpringConstant=75
    bAngularSpring=true
    bGravityOverrideInSimSpace=true
    bOverrideAngularDamping=true
    simulationSpace=RootRelative
    alphaBoolBlend {blendOption=Linear}
    boundBone {boneName=dyn_hair_bk_l_03}
    componentAppliedLinearAccClamp=<10000 10000 8000>
    componentLinearAccScale=<1.4 1.4 1.6>
    componentLinearVelScale=<1.4 1.4 1.6>
    gravityOverride {z=-980}
    physicsBodyDefinitions [{
      sphereCollisionRadius=0
      boundBone {boneName=dyn_hair_bk_l_03}
      boxExtents=<20 20 10>
      constraintSetup {angularLimitsMax=<45 0 45>, angularLimitsMin=<-45 -45 -45>, angularTarget=<0 0 0>}
      localJointOffset=<2 0 -0.5>
     }]
    relativeSpaceBone {boneName=root}
   }
   n15 = Anim Dynamics - Bone: dyn_hair_bk_l_04!(ComponentPose=@n14, Alpha=1, GravityScale:float=1, GravityOverride:Vector=@n5, ExternalForce="0,0,0") -> (Pose) {
    angularDampingOverride=1
    angularSpringConstant=75
    bAngularSpring=true
    bGravityOverrideInSimSpace=true
    bOverrideAngularDamping=true
    simulationSpace=RootRelative
    alphaBoolBlend {blendOption=Linear}
    boundBone {boneName=dyn_hair_bk_l_04}
    componentAppliedLinearAccClamp=<10000 10000 8000>
    componentLinearAccScale=<1.4 1.4 1.6>
    componentLinearVelScale=<1.4 1.4 1.6>
    gravityOverride {z=-980}
    physicsBodyDefinitions [{
      sphereCollisionRadius=0
      boundBone {boneName=dyn_hair_bk_l_04}
      boxExtents=<20 20 10>
      constraintSetup {angularLimitsMax=<45 3 45>, angularLimitsMin=<-45 -45 -45>, angularTarget=<0 0 0>}
      localJointOffset=<2 0 -0.5>
     }]
    relativeSpaceBone {boneName=root}
   }
   n16 = Anim Dynamics - Bone: dyn_hair_bk_l_05!(ComponentPose=@n15, Alpha=1, GravityScale:float=1, GravityOverride:Vector=@n5, ExternalForce="0,0,0") -> (Pose) {
    angularDampingOverride=1
    angularSpringConstant=75
    bAngularSpring=true
    bGravityOverrideInSimSpace=true
    bOverrideAngularDamping=true
    simulationSpace=RootRelative
    alphaBoolBlend {blendOption=Linear}
    boundBone {boneName=dyn_hair_bk_l_05}
    componentAppliedLinearAccClamp=<10000 10000 8000>
    componentLinearAccScale=<1.4 1.4 1.6>
    componentLinearVelScale=<1.4 1.4 1.6>
    gravityOverride {z=-980}
    physicsBodyDefinitions [{
      sphereCollisionRadius=0
      boundBone {boneName=dyn_hair_bk_l_05}
      boxExtents=<20 20 10>
      constraintSetup {angularLimitsMax=<45 7 45>, angularLimitsMin=<-45 -45 -45>, angularTarget=<0 0 0>}
      localJointOffset=<2 0 -0.5>
     }]
    relativeSpaceBone {boneName=root}
   }
   n17 = Anim Dynamics - Bone: dyn_hair_bk_r_02!(ComponentPose=@n16, Alpha=1, GravityScale:float=1, GravityOverride:Vector=@n5, ExternalForce="0,0,0") -> (Pose) {
    angularDampingOverride=1
    bAngularSpring=true
    bGravityOverrideInSimSpace=true
    bOverrideAngularDamping=true
    simulationSpace=RootRelative
    alphaBoolBlend {blendOption=Linear}
    boundBone {boneName=dyn_hair_bk_r_02}
    componentAppliedLinearAccClamp=<5000 5000 6000>
    componentLinearAccScale=<1.2 1.2 1.5>
    componentLinearVelScale=<1.2 1.2 1.5>
    gravityOverride {z=-980}
    physicsBodyDefinitions [{
      sphereCollisionRadius=0
      boundBone {boneName=dyn_hair_bk_r_02}
      boxExtents=<20 20 10>
      constraintSetup {angularLimitsMax=<45 -5 45>, angularLimitsMin=<-45 -45 5>, angularTarget=<0 0 0>}
      localJointOffset=<2 0 -0.5>
     }]
    relativeSpaceBone {boneName=root}
   }
   n18 = Anim Dynamics - Bone: dyn_hair_bk_r_03!(ComponentPose=@n17, Alpha=1, GravityScale:float=1, GravityOverride:Vector=@n5, ExternalForce="0,0,0") -> (Pose) {
    angularDampingOverride=1
    bAngularSpring=true
    bGravityOverrideInSimSpace=true
    bOverrideAngularDamping=true
    simulationSpace=RootRelative
    alphaBoolBlend {blendOption=Linear}
    boundBone {boneName=dyn_hair_bk_r_03}
    componentAppliedLinearAccClamp=<5000 5000 6000>
    componentLinearAccScale=<1.2 1.2 1.5>
    componentLinearVelScale=<1.2 1.2 1.5>
    gravityOverride {z=-980}
    physicsBodyDefinitions [{
      sphereCollisionRadius=0
      boundBone {boneName=dyn_hair_bk_r_03}
      boxExtents=<20 20 10>
      constraintSetup {angularLimitsMax=<45 0 45>, angularLimitsMin=<-45 -45 -35>, angularTarget=<0 0 0>}
      localJointOffset=<2 0 -0.5>
     }]
    relativeSpaceBone {boneName=root}
   }
   n19 = Anim Dynamics - Bone: dyn_hair_bk_r_04!(ComponentPose=@n18, Alpha=1, GravityScale:float=1, GravityOverride:Vector=@n5, ExternalForce="0,0,0") -> (Pose) {
    angularDampingOverride=1
    bAngularSpring=true
    bGravityOverrideInSimSpace=true
    bOverrideAngularDamping=true
    simulationSpace=RootRelative
    alphaBoolBlend {blendOption=Linear}
    boundBone {boneName=dyn_hair_bk_r_04}
    componentAppliedLinearAccClamp=<5000 5000 6000>
    componentLinearAccScale=<1.2 1.2 1.5>
    componentLinearVelScale=<1.2 1.2 1.5>
    gravityOverride {z=-980}
    physicsBodyDefinitions [{
      sphereCollisionRadius=0
      boundBone {boneName=dyn_hair_bk_r_04}
      boxExtents=<20 20 10>
      constraintSetup {angularLimitsMax=<45 -3 45>, angularLimitsMin=<-45 -45 -35>, angularTarget=<0 0 0>}
      localJointOffset=<2 0 -0.5>
     }]
    relativeSpaceBone {boneName=root}
   }
   n20 = Anim Dynamics - Bone: dyn_hair_bk_r_05!(ComponentPose=@n19, Alpha=1, GravityScale:float=1, GravityOverride:Vector=@n5, ExternalForce="0,0,0") -> (Pose) {
    angularDampingOverride=1
    bAngularSpring=true
    bGravityOverrideInSimSpace=true
    bOverrideAngularDamping=true
    simulationSpace=RootRelative
    alphaBoolBlend {blendOption=Linear}
    boundBone {boneName=dyn_hair_bk_r_05}
    componentAppliedLinearAccClamp=<5000 5000 6000>
    componentLinearAccScale=<1.2 1.2 1.5>
    componentLinearVelScale=<1.2 1.2 1.5>
    gravityOverride {z=-980}
    physicsBodyDefinitions [{
      sphereCollisionRadius=0
      boundBone {boneName=dyn_hair_bk_r_05}
      boxExtents=<20 20 10>
      constraintSetup {angularLimitsMax=<45 -5 45>, angularLimitsMin=<-45 -45 -35>, angularTarget=<0 0 0>}
      localJointOffset=<2 0 -0.5>
     }]
    relativeSpaceBone {boneName=root}
   }
   n21 = "Constraint - dyn_hair_bk_l_01 (P)"(ComponentPose=@n20, Alpha=0.5, ConstraintWeights_0=1) {
    alphaBoolBlend {blendOption=Linear}
    boneToModify {boneName=dyn_hair_bk_l_01}
    constraintSetup [{targetBone {boneName=spine_05}}]
    constraintWeights [1]
   }
   n22 = "Constraint - dyn_hair_bk_l_02 (P)"(ComponentPose=@n21, Alpha=0.55, ConstraintWeights_0=1) {
    alphaBoolBlend {blendOption=Linear}
    boneToModify {boneName=dyn_hair_bk_l_02}
    constraintSetup [{targetBone {boneName=clavicle_l}}]
    constraintWeights [1]
   }
   n23 = "Constraint - dyn_hair_bk_l_03 (PP)"!(ComponentPose=@n22, Alpha=0.5, ConstraintWeights_0=0.25, ConstraintWeights_1:float=0.75) -> (Pose) {
    alphaBoolBlend {blendOption=Linear}
    boneToModify {boneName=dyn_hair_bk_l_03}
    constraintSetup [{targetBone {boneName=spine_05}}, {targetBone {boneName=clavicle_l}}]
    constraintWeights [1,1]
   }
   n24 = "Constraint - dyn_hair_bk_l_04 (P)"(ComponentPose=@n23, Alpha=0.3, ConstraintWeights_0=1) {
    alphaBoolBlend {blendOption=Linear}
    boneToModify {boneName=dyn_hair_bk_l_04}
    constraintSetup [{targetBone {boneName=spine_05}}]
    constraintWeights [1]
   }
   n25 = "Constraint - dyn_hair_bk_01 (P)"(ComponentPose=@n24, Alpha=0.6, ConstraintWeights_0=1) {
    alphaBoolBlend {blendOption=Linear}
    boneToModify {boneName=dyn_hair_bk_01}
    constraintSetup [{targetBone {boneName=spine_05}}]
    constraintWeights [1]
   }
   n26 = "Constraint - dyn_hair_bk_02 (P)"(ComponentPose=@n25, Alpha=0.2, ConstraintWeights_0=1) {
    alphaBoolBlend {blendOption=Linear}
    boneToModify {boneName=dyn_hair_bk_02}
    constraintSetup [{targetBone {boneName=spine_05}}]
    constraintWeights [1]
   }
   n27 = "Constraint - dyn_hair_bk_03 (P)"(ComponentPose=@n26, Alpha=0.2, ConstraintWeights_0=1) {
    alphaBoolBlend {blendOption=Linear}
    boneToModify {boneName=dyn_hair_bk_03}
    constraintSetup [{targetBone {boneName=spine_05}}]
    constraintWeights [1]
   }
   n28 = "Constraint - dyn_hair_bk_04 (P)"(ComponentPose=@n27, Alpha=0.1, ConstraintWeights_0=1) {
    alphaBoolBlend {blendOption=Linear}
    boneToModify {boneName=dyn_hair_bk_04}
    constraintSetup [{targetBone {boneName=spine_05}}]
    constraintWeights [1]
   }
   n29 = "Constraint - dyn_hair_bk_05 (P)"(ComponentPose=@n28, Alpha=0.1, ConstraintWeights_0=1) {
    alphaBoolBlend {blendOption=Linear}
    boneToModify {boneName=dyn_hair_bk_05}
    constraintSetup [{targetBone {boneName=spine_04}}]
    constraintWeights [1]
   }
   n30 = "Constraint - dyn_hair_bk_06 (P)"(ComponentPose=@n29, Alpha=0.1, ConstraintWeights_0=1) {
    alphaBoolBlend {blendOption=Linear}
    boneToModify {boneName=dyn_hair_bk_06}
    constraintSetup [{targetBone {boneName=spine_04}}]
    constraintWeights [1]
   }
   n31 = "Constraint - dyn_hair_bk_06 (P)"(ComponentPose=@n30, Alpha=0.1, ConstraintWeights_0=1) {
    alphaBoolBlend {blendOption=Linear}
    boneToModify {boneName=dyn_hair_bk_06}
    constraintSetup [{targetBone {boneName=spine_04}}]
    constraintWeights [1]
   }
   n32 = "Constraint - dyn_hair_bk_r_03 (P)"(ComponentPose=@n31, Alpha=0.4, ConstraintWeights_0=1) {
    alphaBoolBlend {blendOption=Linear}
    boneToModify {boneName=dyn_hair_bk_r_03}
    constraintSetup [{targetBone {boneName=clavicle_r}}]
    constraintWeights [1]
   }
   n33 = "Constraint - dyn_hair_bk_r_04 (P)"(ComponentPose=@n32, Alpha=0.3, ConstraintWeights_0=1) {
    alphaBoolBlend {blendOption=Linear}
    boneToModify {boneName=dyn_hair_bk_r_04}
    constraintSetup [{targetBone {boneName=spine_05}}]
    constraintWeights [1]
   }
   n34 = Get skydive Hair ROT() -> (skydive Hair ROT:bool)
   n35 = "Transform (Modify) Bone - Bone: dyn_hair_bk_l_02"!(ComponentPose=@n33, bAlphaBoolEnabled:bool=@n34, Translation="0,0,0", Rotation="5,0.000,0.000", Scale="1,1,1") -> (Pose) {
    alphaInputType=Bool
    rotationMode=BMM_Additive
    rotationSpace=BCS_BoneSpace
    scaleMode=BMM_Additive
    scaleSpace=BCS_BoneSpace
    translationMode=BMM_Additive
    translationSpace=BCS_BoneSpace
    alphaBoolBlend {blendInTime=0.1, blendOption=Linear, blendOutTime=0.1}
    boneToModify {boneName=dyn_hair_bk_l_02}
    rotation {pitch=5}
   }
   n36 = "Transform (Modify) Bone - Bone: dyn_hair_bk_03"!(ComponentPose=@n35, bAlphaBoolEnabled:bool=@n34, Translation="0,0,0", Rotation="10,0.000,0.000", Scale="1,1,1") -> (Pose) {
    alphaInputType=Bool
    rotationMode=BMM_Additive
    rotationSpace=BCS_BoneSpace
    scaleMode=BMM_Additive
    scaleSpace=BCS_BoneSpace
    translationMode=BMM_Additive
    translationSpace=BCS_BoneSpace
    alphaBoolBlend {blendInTime=0.1, blendOption=Linear, blendOutTime=0.1}
    boneToModify {boneName=dyn_hair_bk_03}
    rotation {pitch=10}
   }
   n37 = "Transform (Modify) Bone - Bone: dyn_hair_bk_02"!(ComponentPose=@n36, bAlphaBoolEnabled:bool=@n34, Translation="0,0,0", Rotation="5,0,0", Scale="1,1,1") -> (Pose) {
    alphaInputType=Bool
    rotationMode=BMM_Additive
    rotationSpace=BCS_BoneSpace
    scaleMode=BMM_Additive
    scaleSpace=BCS_BoneSpace
    translationMode=BMM_Additive
    translationSpace=BCS_BoneSpace
    alphaBoolBlend {blendInTime=0.1, blendOption=Linear, blendOutTime=0.1}
    boneToModify {boneName=dyn_hair_bk_02}
    rotation {pitch=5}
   }
   n38 = "Transform (Modify) Bone - Bone: dyn_hair_bk_03"(ComponentPose=@n37, bAlphaBoolEnabled=@n34, Translation="0,0,0", Rotation="5,0,0", Scale="1,1,1") {
    alphaInputType=Bool
    rotationMode=BMM_Additive
    rotationSpace=BCS_BoneSpace
    scaleMode=BMM_Additive
    scaleSpace=BCS_BoneSpace
    translationMode=BMM_Additive
    translationSpace=BCS_BoneSpace
    alphaBoolBlend {blendInTime=0.1, blendOption=Linear, blendOutTime=0.1}
    boneToModify {boneName=dyn_hair_bk_03}
    rotation {pitch=5}
   }
   n39 = "Transform (Modify) Bone - Bone: dyn_hair_bk_r_02"!(ComponentPose=@n38, bAlphaBoolEnabled:bool=@n34, Translation="0,0,0", Rotation="10,5.319,1.817", Scale="1,1,1") -> (Pose) {
    alphaInputType=Bool
    rotationMode=BMM_Additive
    rotationSpace=BCS_BoneSpace
    scaleMode=BMM_Additive
    scaleSpace=BCS_BoneSpace
    translationMode=BMM_Additive
    translationSpace=BCS_BoneSpace
    alphaBoolBlend {blendOption=Linear}
    boneToModify {boneName=dyn_hair_bk_r_02}
    rotation=<rot 10 5.319 1.817>
   }
   n40 = "Transform (Modify) Bone - Bone: dyn_hair_bk_r_03"!(ComponentPose=@n39, bAlphaBoolEnabled:bool=@n34, Translation="0,0,0", Rotation="5,21.171,7.096", Scale="1,1,1") -> (Pose) {
    alphaInputType=Bool
    rotationMode=BMM_Additive
    rotationSpace=BCS_BoneSpace
    scaleMode=BMM_Additive
    scaleSpace=BCS_BoneSpace
    translationMode=BMM_Additive
    translationSpace=BCS_BoneSpace
    alphaBoolBlend {blendOption=Linear}
    boneToModify {boneName=dyn_hair_bk_r_03}
    rotation=<rot 5 21.171 7.096>
   }
   n41 = Anim Dynamics - Bone: dyn_hair_bk_l_03(
    ComponentPose=Local To Component(LocalPose=Use cached pose 'prebackpack'())
    Alpha=1
    GravityScale=1
    GravityOverride="0,0,-980"
    ExternalForce="0,0,0"
   ) {
    angularDampingOverride=1
    angularSpringConstant=75
    bAngularSpring=true
    bGravityOverrideInSimSpace=true
    bOverrideAngularDamping=true
    simulationSpace=RootRelative
    alphaBoolBlend {blendOption=Linear}
    boundBone {boneName=dyn_hair_bk_l_03}
    componentAppliedLinearAccClamp=<10000 10000 8000>
    componentLinearAccScale=<1.4 1.4 1.6>
    componentLinearVelScale=<1.4 1.4 1.6>
    gravityOverride {z=-980}
    physicsBodyDefinitions [{
      sphereCollisionRadius=0
      boundBone {boneName=dyn_hair_bk_l_03}
      boxExtents=<20 20 10>
      constraintSetup {angularLimitsMax=<45 0 45>, angularLimitsMin=<-45 -45 -45>, angularTarget=<0 0 0>}
      localJointOffset=<2 0 -0.5>
     }]
    relativeSpaceBone {boneName=root}
   }
   n42 = Anim Dynamics - Bone: dyn_hair_bk_l_04(ComponentPose=@n41, Alpha=1, GravityScale=1, GravityOverride="0,0,-980", ExternalForce="0,0,0") {
    angularDampingOverride=1
    angularSpringConstant=75
    bAngularSpring=true
    bGravityOverrideInSimSpace=true
    bOverrideAngularDamping=true
    simulationSpace=RootRelative
    alphaBoolBlend {blendOption=Linear}
    boundBone {boneName=dyn_hair_bk_l_04}
    componentAppliedLinearAccClamp=<10000 10000 8000>
    componentLinearAccScale=<1.4 1.4 1.6>
    componentLinearVelScale=<1.4 1.4 1.6>
    gravityOverride {z=-980}
    physicsBodyDefinitions [{
      sphereCollisionRadius=0
      boundBone {boneName=dyn_hair_bk_l_04}
      boxExtents=<20 20 10>
      constraintSetup {angularLimitsMax=<45 3 45>, angularLimitsMin=<-45 -45 -45>, angularTarget=<0 0 0>}
      localJointOffset=<2 0 -0.5>
     }]
    relativeSpaceBone {boneName=root}
   }
   n43 = Anim Dynamics - Bone: dyn_hair_bk_l_05(ComponentPose=@n42, Alpha=1, GravityScale=1, GravityOverride="0,0,-980", ExternalForce="0,0,0") {
    angularDampingOverride=1
    angularSpringConstant=75
    bAngularSpring=true
    bGravityOverrideInSimSpace=true
    bOverrideAngularDamping=true
    simulationSpace=RootRelative
    alphaBoolBlend {blendOption=Linear}
    boundBone {boneName=dyn_hair_bk_l_05}
    componentAppliedLinearAccClamp=<10000 10000 8000>
    componentLinearAccScale=<1.4 1.4 1.6>
    componentLinearVelScale=<1.4 1.4 1.6>
    gravityOverride {z=-980}
    physicsBodyDefinitions [{
      sphereCollisionRadius=0
      boundBone {boneName=dyn_hair_bk_l_05}
      boxExtents=<20 20 10>
      constraintSetup {angularLimitsMax=<45 7 45>, angularLimitsMin=<-45 -45 -45>, angularTarget=<0 0 0>}
      localJointOffset=<2 0 -0.5>
     }]
    relativeSpaceBone {boneName=root}
   }
   n44 = Anim Dynamics - Bone: dyn_hair_bk_r_04(ComponentPose=@n43, Alpha=1, GravityScale=1, GravityOverride="0,0,-980", ExternalForce="0,0,0") {
    angularDampingOverride=1
    bAngularSpring=true
    bGravityOverrideInSimSpace=true
    bOverrideAngularDamping=true
    simulationSpace=RootRelative
    alphaBoolBlend {blendOption=Linear}
    boundBone {boneName=dyn_hair_bk_r_04}
    componentAppliedLinearAccClamp=<5000 5000 6000>
    componentLinearAccScale=<1.2 1.2 1.5>
    componentLinearVelScale=<1.2 1.2 1.5>
    gravityOverride {z=-980}
    physicsBodyDefinitions [{
      sphereCollisionRadius=0
      boundBone {boneName=dyn_hair_bk_r_04}
      boxExtents=<20 20 10>
      constraintSetup {angularLimitsMax=<45 -3 45>, angularLimitsMin=<-45 -45 -35>, angularTarget=<0 0 0>}
      localJointOffset=<2 0 -0.5>
     }]
    relativeSpaceBone {boneName=root}
   }
   n45 = Anim Dynamics - Bone: dyn_hair_bk_r_05(ComponentPose=@n44, Alpha=1, GravityScale=1, GravityOverride="0,0,-980", ExternalForce="0,0,0") {
    angularDampingOverride=1
    bAngularSpring=true
    bGravityOverrideInSimSpace=true
    bOverrideAngularDamping=true
    simulationSpace=RootRelative
    alphaBoolBlend {blendOption=Linear}
    boundBone {boneName=dyn_hair_bk_r_05}
    componentAppliedLinearAccClamp=<5000 5000 6000>
    componentLinearAccScale=<1.2 1.2 1.5>
    componentLinearVelScale=<1.2 1.2 1.5>
    gravityOverride {z=-980}
    physicsBodyDefinitions [{
      sphereCollisionRadius=0
      boundBone {boneName=dyn_hair_bk_r_05}
      boxExtents=<20 20 10>
      constraintSetup {angularLimitsMax=<45 -5 45>, angularLimitsMin=<-45 -45 -35>, angularTarget=<0 0 0>}
      localJointOffset=<2 0 -0.5>
     }]
    relativeSpaceBone {boneName=root}
   }
   n46 = "Constraint - dyn_hair_bk_l_01 (P)"(ComponentPose=@n45, Alpha=0.5, ConstraintWeights_0=1) {
    alphaBoolBlend {blendOption=Linear}
    boneToModify {boneName=dyn_hair_bk_l_01}
    constraintSetup [{targetBone {boneName=spine_05}}]
    constraintWeights [1]
   }
   n47 = "Constraint - dyn_hair_bk_l_02 (P)"(ComponentPose=@n46, Alpha=0.55, ConstraintWeights_0=1) {
    alphaBoolBlend {blendOption=Linear}
    boneToModify {boneName=dyn_hair_bk_l_02}
    constraintSetup [{targetBone {boneName=clavicle_l}}]
    constraintWeights [1]
   }
   n48 = "Constraint - dyn_hair_bk_l_03 (PP)"(ComponentPose=@n47, Alpha=0.5, ConstraintWeights_0=0.25, ConstraintWeights_1=0.75) {
    alphaBoolBlend {blendOption=Linear}
    boneToModify {boneName=dyn_hair_bk_l_03}
    constraintSetup [{targetBone {boneName=spine_05}}, {targetBone {boneName=clavicle_l}}]
    constraintWeights [1,1]
   }
   n49 = "Constraint - dyn_hair_bk_l_04 (P)"(ComponentPose=@n48, Alpha=0.3, ConstraintWeights_0=1) {
    alphaBoolBlend {blendOption=Linear}
    boneToModify {boneName=dyn_hair_bk_l_04}
    constraintSetup [{targetBone {boneName=spine_05}}]
    constraintWeights [1]
   }
   n50 = "Constraint - dyn_hair_bk_01 (P)"(ComponentPose=@n49, Alpha=0.6, ConstraintWeights_0=1) {
    alphaBoolBlend {blendOption=Linear}
    boneToModify {boneName=dyn_hair_bk_01}
    constraintSetup [{targetBone {boneName=spine_05}}]
    constraintWeights [1]
   }
   n51 = "Constraint - dyn_hair_bk_02 (P)"(ComponentPose=@n50, Alpha=0.2, ConstraintWeights_0=1) {
    alphaBoolBlend {blendOption=Linear}
    boneToModify {boneName=dyn_hair_bk_02}
    constraintSetup [{targetBone {boneName=spine_05}}]
    constraintWeights [1]
   }
   n52 = "Constraint - dyn_hair_bk_03 (P)"(ComponentPose=@n51, Alpha=0.2, ConstraintWeights_0=1) {
    alphaBoolBlend {blendOption=Linear}
    boneToModify {boneName=dyn_hair_bk_03}
    constraintSetup [{targetBone {boneName=spine_05}}]
    constraintWeights [1]
   }
   n53 = "Constraint - dyn_hair_bk_04 (P)"(ComponentPose=@n52, Alpha=0.1, ConstraintWeights_0=1) {
    alphaBoolBlend {blendOption=Linear}
    boneToModify {boneName=dyn_hair_bk_04}
    constraintSetup [{targetBone {boneName=spine_05}}]
    constraintWeights [1]
   }
   n54 = "Constraint - dyn_hair_bk_05 (P)"(ComponentPose=@n53, Alpha=0.1, ConstraintWeights_0=1) {
    alphaBoolBlend {blendOption=Linear}
    boneToModify {boneName=dyn_hair_bk_05}
    constraintSetup [{targetBone {boneName=spine_04}}]
    constraintWeights [1]
   }
   n55 = "Constraint - dyn_hair_bk_06 (P)"(ComponentPose=@n54, Alpha=0.1, ConstraintWeights_0=1) {
    alphaBoolBlend {blendOption=Linear}
    boneToModify {boneName=dyn_hair_bk_06}
    constraintSetup [{targetBone {boneName=spine_04}}]
    constraintWeights [1]
   }
   n56 = "Constraint - dyn_hair_bk_06 (P)"(ComponentPose=@n55, Alpha=0.1, ConstraintWeights_0=1) {
    alphaBoolBlend {blendOption=Linear}
    boneToModify {boneName=dyn_hair_bk_06}
    constraintSetup [{targetBone {boneName=spine_04}}]
    constraintWeights [1]
   }
   n57 = "Constraint - dyn_hair_bk_r_03 (P)"(ComponentPose=@n56, Alpha=0.4, ConstraintWeights_0=1) {
    alphaBoolBlend {blendOption=Linear}
    boneToModify {boneName=dyn_hair_bk_r_03}
    constraintSetup [{targetBone {boneName=clavicle_r}}]
    constraintWeights [1]
   }
   n58 = "Constraint - dyn_hair_bk_r_04 (P)"(ComponentPose=@n57, Alpha=0.3, ConstraintWeights_0=1) {
    alphaBoolBlend {blendOption=Linear}
    boneToModify {boneName=dyn_hair_bk_r_04}
    constraintSetup [{targetBone {boneName=spine_05}}]
    constraintWeights [1]
   }
   Output Pose(Result=Blend Poses by bool(
     BlendPose_0=Component To Local(ComponentPose=@n58)
     BlendPose_1=Component To Local(ComponentPose=@n40)
     bActiveValue=Get wearingBackpack() -> (wearingBackpack:bool)
    ) {blendType=Linear}) {layerGroup=None}
   n59 = Get Skel_Mesh()
   n60 = Look At - Bone: dyn_hair_bk_base(ComponentPose=Local To Component(LocalPose=Copy Pose From Mesh(SourceMeshComponent=@n59) {bUseAttachedParent=true}), Alpha=0.8) {
    alphaBoolBlend {blendOption=Linear}
    boneToModify {boneName=dyn_hair_bk_base}
    lookAtLocation {x=0}
    lookAtTarget {boneReference {boneName=dyn_hair_bk_MSTR}}
    lookAt_Axis {axis=<1 0>}
   }
   n61 = Look At - Bone: dyn_hair_r_Lookat(ComponentPose=@n60, Alpha=0.8) {
    alphaBoolBlend {blendOption=Linear}
    boneToModify {boneName=dyn_hair_r_Lookat}
    lookAtLocation {x=0}
    lookAtTarget {boneReference {boneName=root}}
    lookAt_Axis {axis=<1 0>}
   }
   n62 = Look At - Bone: dyn_hair_l_Lookat(ComponentPose=@n61, Alpha=0.8) {
    alphaBoolBlend {blendOption=Linear}
    boneToModify {boneName=dyn_hair_l_Lookat}
    lookAtLocation {x=0}
    lookAtTarget {boneReference {boneName=root}}
    lookAt_Axis {axis=<1 0>}
   }
   n63 = RigidBody!(ComponentPose=@n62, Alpha, ExternalForce) -> (Pose) {
    bClampLinearTranslationLimitToRefPose=true
    bOverrideWorldGravity=true
    overridePhysicsAsset=&/F_MED_SAM_dyn_hair_temple
    simulationSpace=BaseBoneSpace
    alphaBoolBlend {blendOption=Linear}
    baseBoneRef {boneName=spine_05}
    componentAppliedLinearAccClamp=<200 200 500>
    componentLinearAccScale=<0.1 0.1 0.15>
    componentLinearVelScale=<0.1 0.1 0.1>
    overrideWorldGravity {z=-980}
   }
   n64 = Anim Dynamics - Bone: dyn_hair_bk_base(ComponentPose=@n63, Alpha=0.5, ExternalForce="0,0,0") {
    angularDampingOverride=0.9999
    angularSpringConstant=100
    bGravityOverrideInSimSpace=true
    bOverrideAngularDamping=true
    simulationSpace=RootRelative
    alphaBoolBlend {blendOption=Linear}
    boundBone {boneName=dyn_hair_bk_base}
    componentAppliedLinearAccClamp=<3500 3500 3500>
    componentLinearAccScale=<0.6 0.6 0.6>
    componentLinearVelScale=<0.6 0.6 0.6>
    physicsBodyDefinitions [{
      sphereCollisionRadius=0
      boundBone {boneName=dyn_hair_bk_base}
      boxExtents=<20 20 10>
      constraintSetup {angularLimitsMax=<5 0 15>, angularLimitsMin=<-5 -25 -15>, angularTarget=<0 0 0>}
      localJointOffset=<1 0 0>
     }]
    relativeSpaceBone {boneName=root}
   }
   n65 = Component To Local(ComponentPose=@n64)
   n66 = Get Skel_Mesh()
   n67 = Look At - Bone: dyn_hair_bk_base(ComponentPose=Local To Component(LocalPose=Copy Pose From Mesh(SourceMeshComponent=@n66) {bUseAttachedParent=true}), Alpha=0.8) {
    alphaBoolBlend {blendOption=Linear}
    boneToModify {boneName=dyn_hair_bk_base}
    lookAtLocation {x=0}
    lookAtTarget {boneReference {boneName=dyn_hair_bk_MSTR}}
    lookAt_Axis {axis=<1 0>}
   }
   n68 = Look At - Bone: dyn_hair_r_Lookat(ComponentPose=@n67, Alpha=0.8) {
    alphaBoolBlend {blendOption=Linear}
    boneToModify {boneName=dyn_hair_r_Lookat}
    lookAtLocation {x=0}
    lookAtTarget {boneReference {boneName=root}}
    lookAt_Axis {axis=<1 0>}
   }
   n69 = Look At - Bone: dyn_hair_l_Lookat(ComponentPose=@n68, Alpha=0.8) {
    alphaBoolBlend {blendOption=Linear}
    boneToModify {boneName=dyn_hair_l_Lookat}
    lookAtLocation {x=0}
    lookAtTarget {boneReference {boneName=root}}
    lookAt_Axis {axis=<1 0>}
   }
   n70 = RigidBody(ComponentPose=@n69) {
    bClampLinearTranslationLimitToRefPose=true
    bOverrideWorldGravity=true
    overridePhysicsAsset=&/F_MED_SAM_dyn_hair_temple_frontend
    simulationSpace=BaseBoneSpace
    alphaBoolBlend {blendOption=Linear}
    baseBoneRef {boneName=spine_05}
    componentAppliedLinearAccClamp=<200 200 500>
    componentLinearAccScale=<0.1 0.1 0.15>
    componentLinearVelScale=<0.1 0.1 0.1>
    overrideWorldGravity {z=-980}
   }
   n71 = Anim Dynamics - Bone: dyn_hair_bk_base(ComponentPose=@n70, Alpha=0.5, ExternalForce="0,0,0") {
    angularDampingOverride=0.9999
    angularSpringConstant=100
    bGravityOverrideInSimSpace=true
    bOverrideAngularDamping=true
    simulationSpace=RootRelative
    alphaBoolBlend {blendOption=Linear}
    boundBone {boneName=dyn_hair_bk_base}
    componentAppliedLinearAccClamp=<3500 3500 3500>
    componentLinearAccScale=<0.6 0.6 0.6>
    componentLinearVelScale=<0.6 0.6 0.6>
    physicsBodyDefinitions [{
      sphereCollisionRadius=0
      boundBone {boneName=dyn_hair_bk_base}
      boxExtents=<20 20 10>
      constraintSetup {angularLimitsMax=<5 0 15>, angularLimitsMin=<-5 -25 -15>, angularTarget=<0 0 0>}
      localJointOffset=<1 0 0>
     }]
    relativeSpaceBone {boneName=root}
   }
   n72 = Component To Local(ComponentPose=@n71)
   n73 = Get bIsInFrontEnd()
   n74 = NOT Boolean(A=Get bRigidBodyEnabled())
   n75 = Look At - Bone: dyn_hair_r_out_01!(
    ComponentPose=Local To Component(LocalPose=Blend Poses by bool(BlendPose_0=@n72, BlendPose_1=@n65, bActiveValue=@n73) {blendType=Linear})
    bAlphaBoolEnabled:bool=@n74
   ) -> (Pose) {
    alphaInputType=Bool
    alphaBoolBlend {blendOption=Linear}
    boneToModify {boneName=dyn_hair_r_out_01}
    lookAtLocation=<0 8 10.0267>
    lookAtTarget {boneReference {boneName=spine_05}}
    lookAt_Axis {axis=<1 0>}
   }
   n76 = Look At - Bone: dyn_hair_r_in_02!(ComponentPose=@n75, bAlphaBoolEnabled:bool=@n74) -> (Pose) {
    alphaInputType=Bool
    alphaBoolBlend {blendOption=Linear}
    boneToModify {boneName=dyn_hair_r_in_02}
    lookAtLocation=<-18.2718 16.8418 7.6051>
    lookAtTarget {boneReference {boneName=spine_05}}
    lookAt_Axis {axis=<1 0>}
   }
   n77 = Look At - Bone: dyn_hair_r_in_04!(ComponentPose=@n76, bAlphaBoolEnabled:bool=@n74) -> (Pose) {
    alphaInputType=Bool
    alphaBoolBlend {blendOption=Linear}
    boneToModify {boneName=dyn_hair_r_in_04}
    lookAtLocation=<-18.2718 20.4908 9.0993>
    lookAtTarget {boneReference {boneName=spine_05}}
    lookAt_Axis {axis=<1 0>}
   }
   n78 = Look At - Bone: dyn_hair_l_02!(ComponentPose=@n77, bAlphaBoolEnabled:bool=@n74) -> (Pose) {
    alphaInputType=Bool
    alphaBoolBlend {blendOption=Linear}
    boneToModify {boneName=dyn_hair_l_02}
    lookAtLocation=<-20 4.0306 -7.1915>
    lookAtTarget {boneReference {boneName=spine_05}}
    lookAt_Axis {axis=<1 0>}
   }
   n79 = Look At - Bone: dyn_hair_l_04!(ComponentPose=@n78, bAlphaBoolEnabled:bool=@n74) -> (Pose) {
    alphaInputType=Bool
    alphaBoolBlend {blendOption=Linear}
    boneToModify {boneName=dyn_hair_l_04}
    lookAtLocation=<-20 20.4856 -7.2617>
    lookAtTarget {boneReference {boneName=spine_05}}
    lookAt_Axis {axis=<1 0>}
   }
   n80 = Look At - Bone: dyn_hair_l_06!(ComponentPose=@n79, bAlphaBoolEnabled:bool=@n74) -> (Pose) {
    alphaInputType=Bool
    alphaBoolBlend {blendOption=Linear}
    boneToModify {boneName=dyn_hair_l_06}
    lookAtLocation=<-20 14.7253 -10.5007>
    lookAtTarget {boneReference {boneName=spine_04}}
    lookAt_Axis {axis=<1 0>}
   }
   Save cached pose 'prebackpack'(Pose=Component To Local(ComponentPose=@n80)) {cacheName=prebackpack}
  ]}
 EventGraph {nodes [
   Event BlueprintUpdateAnimation()
   > n1 = Sequence!(execute) -> (then_0, then_1, then_2:Exec)
   n2 = BreakTransform(InTransform=GetDeltaTransformFromRefPose(self=GetOwnerSkeletalMeshComponent(), BoneName=head, BaseName=spine_05))
   > Set head_rot_up(head_rot_up:double=MapRangeClamped(Value=BreakRotator(InRot=@n2.Rotation).Yaw, InRangeA=-15, InRangeB=-45)) -> (Output_Get:double)
   > n3 = Gravity Override State Function DEPRECATED(CharacterPartAnimInstance=Self-Reference(), GameTimeInSeconds=GetGameTimeInSeconds(), GravityOverrideParams=Get Gravity Override Params())
   > Set gravity_override(gravity_override=@n3)
   n4 = Get bIsParachuteOpened()
   n5 = Branch(execute=@n1.then_1, Condition=AND Boolean(A=Get bIsSkyDiving(), B=NOT Boolean(A=@n4)))
   > Set skydive Hair ROT(skydive Hair ROT:bool=true) -> (Output_Get:bool)
   Set skydive Hair ROT(execute=@n5.else, skydive Hair ROT=false)
   Set wearingBackpack(execute=@n1.then_2, wearingBackpack:bool=CheckForBackbling()) -> (Output_Get:bool)
   Event BlueprintInitializeAnimation()
   > Set Skel_Mesh(Skel_Mesh=GetOwnerSkeletalMeshComponent())
   TryGetPawnOwner()
  ]}
]
variables [
 Gravity Override Params {type=GravityOverrideParamsStruct}
 gravity_override {type=Vector}
 head_rot_up {type=number}
 isNotWearingBackbling {type=number}
 isWearingBackpack {type=number}
 isWearingCape {type=number}
 Skel_Mesh {type=SkeletalMeshComponent}
 skydive Hair ROT {type=boolean}
 wearingBackpack {type=boolean}
 WearingCape {type=boolean}
]
